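# Post-process style AnimBP for the F_MED_PastaSauceMarinara cosmetic body that copies the parent mesh pose and animates its straps, horns, chains, neck ring, bows and laces.
# Weighted bone constraints pin straps and horns to spine, neck, shoulder and arm bones, some repeated after simulation.
# A Control Rig takes gameplay state such as skydive, weapon, ADS and crouch.
# AnimDynamics springs add secondary motion, with chains using spherical limits at the right leg.
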AnimBlueprint /BRCosmetics/Characters/Player/Female/Medium/Bodies/F_MED_PastaSauceMarinara/Meshes/F_MED_PastaSauceMarinara_AnimBP
target_skeleton=&/F_MED_PastaSauceMarinara_Skeleton
graphs [
 AnimGraph {nodes [
   n1 = "Constraint - dyn_strap_fr_Top (PP)"!(
    ComponentPose=Local To Component(LocalPose=Copy Pose From Mesh() {bCopyCurves=true, bUseAttachedParent=true})
    Alpha=1
    ConstraintWeights_0=0.5
    ConstraintWeights_1:float=0.5
   ) -> (Pose) {
    alphaBoolBlend {blendOption=Linear}
    boneToModify {boneName=dyn_strap_fr_Top}
    constraintSetup [{targetBone {boneName=dyn_shoulder_01}}, {targetBone {boneName=dyn_shoulder_02}}]
    constraintWeights [1,1]
   }
   n2 = "Constraint - dyn_strap_fr_C (PPPP)"!(
    ComponentPose=@n1
    Alpha=1
    ConstraintWeights_0=0.4472
    ConstraintWeights_1:float=0.0501
    ConstraintWeights_2:float=0.0957
    ConstraintWeights_3:float=0.407
   ) -> (Pose) {
    alphaBoolBlend {blendOption=Linear}
    boneToModify {boneName=dyn_strap_fr_C}
    constraintSetup [
     {targetBone {boneName=upperarm_twist_01_l}}
     {targetBone {boneName=spine_04}}
     {targetBone {boneName=spine_05}}
     {targetBone {boneName=clavicle_l}}
    ]
    constraintWeights [1,1,1,1]
   }
   n3 = "Constraint - dyn_strap_fr_Bot (PPPP)"!(
    ComponentPose=@n2
    Alpha=1
    ConstraintWeights_0=0.1188
    ConstraintWeights_1:float=0.2131
    ConstraintWeights_2:float=0.2547
    ConstraintWeights_3:float=0.4134
   ) -> (Pose) {
    alphaBoolBlend {blendOption=Linear}
    boneToModify {boneName=dyn_strap_fr_Bot}
    constraintSetup [
     {targetBone {boneName=upperarm_twist_01_l}}
     {targetBone {boneName=spine_04}}
     {targetBone {boneName=spine_05}}
     {targetBone {boneName=clavicle_l}}
    ]
    constraintWeights [1,1,1,1]
   }
   n4 = "Constraint - dyn_strap_bk_Top (PP)"!(ComponentPose=@n3, Alpha=1, ConstraintWeights_0=0.5, ConstraintWeights_1:float=0.5) -> (Pose) {
    alphaBoolBlend {blendOption=Linear}
    boneToModify {boneName=dyn_strap_bk_Top}
    constraintSetup [{targetBone {boneName=dyn_shoulder_01}}, {targetBone {boneName=dyn_shoulder_02}}]
    constraintWeights [1,1]
   }
   n5 = "Constraint - dyn_strap_bk_C (PPP)"!(ComponentPose=@n4, Alpha=1, ConstraintWeights_0=0.2509, ConstraintWeights_1:float=0.031, ConstraintWeights_2:float=0.7181) -> (Pose) {
    alphaBoolBlend {blendOption=Linear}
    boneToModify {boneName=dyn_strap_bk_C}
    constraintSetup [{targetBone {boneName=upperarm_twist_01_l}}, {targetBone {boneName=spine_05}}, {targetBone {boneName=clavicle_l}}]
    constraintWeights [1,1,1]
   }
   n6 = "Constraint - dyn_strap_bk_Bot (PPPPPPP)"!(
    ComponentPose=@n5
    Alpha=1
    ConstraintWeights_0=0.0952
    ConstraintWeights_1:float=0
    ConstraintWeights_2:float=0.0347
    ConstraintWeights_3:float=0
    ConstraintWeights_4:float=0.1824
    ConstraintWeights_5:float=0.6863
    ConstraintWeights_6:float=0.0014
   ) -> (Pose) {
    alphaBoolBlend {blendOption=Linear}
    boneToModify {boneName=dyn_strap_bk_Bot}
    constraintSetup [
     {targetBone {boneName=upperarm_twist_01_l}}
     {targetBone {boneName=spine_02}}
     {targetBone {boneName=spine_04}}
     {targetBone {boneName=spine_03}}
     {targetBone {boneName=spine_05}}
     {targetBone {boneName=clavicle_l}}
     {targetBone {boneName=clavicle_r}}
    ]
    constraintWeights [1,1,1,1,1,1,1]
   }
   n7 = Control Rig!(
    Source=Component To Local(ComponentPose=@n6)
    Alpha
    SkyDive:double=Get Skydive() -> (Skydive:double)
    Pistol:double=Get pistol() -> (pistol:double)
    Targeting:double=Get ADS() -> (ADS:double)
    Melee:double=Get Melee() -> (Melee:double)
    Moving:double=Get moving() -> (moving:double)
    Crouch:double=Get Crouch() -> (Crouch:double)
    Parachute:double=Get parachute() -> (parachute:double)
    Sky:bool=Get bIsSkyDiving()
    Rifle:double=Get Rifle() -> (Rifle:double)
   ) -> (Pose) {
    alphaBoolBlend {blendOption=Linear}
    controlRigAssetReference {blueprintRigClass=&/F_MED_PastaSauceMarinara_CtrlRig.F_MED_PastaSauceMarinara_CtrlRig_C}
   }
   n8 = "Constraint - dyn_horn_l_05 (PPPPPP)"!(
    ComponentPose=Local To Component(LocalPose=@n7)
    Alpha=0.3
    ConstraintWeights_0=0.2036
    ConstraintWeights_1:float=0.0842
    ConstraintWeights_2:float=0.0645
    ConstraintWeights_3:float=0.1476
    ConstraintWeights_4:float=0.25
    ConstraintWeights_5:float=0.25
   ) -> (Pose) {
    alphaBoolBlend {blendOption=Linear}
    alphaScaleBiasClamp {bInterpResult=true, interpSpeedDecreasing=12, interpSpeedIncreasing=15}
    boneToModify {boneName=dyn_horn_l_05}
    constraintSetup [
     {targetBone {boneName=spine_05}}
     {targetBone {boneName=neck_01}}
     {targetBone {boneName=head}}
     {targetBone {boneName=neck_02}}
     {targetBone {boneName=dyn_shoulder_01}}
     {targetBone {boneName=dyn_shoulder_02}}
    ]
    constraintWeights [1,1,1,1,1,1]
   }
   n9 = "Constraint - dyn_horn_r_05 (PPPPPP)"!(
    ComponentPose=@n8
    Alpha=0.3
    ConstraintWeights_0=0.2995
    ConstraintWeights_1:float=0.2888
    ConstraintWeights_2:float=0.2075
    ConstraintWeights_3:float=0.0684
    ConstraintWeights_4:float=0.033
    ConstraintWeights_5:float=0.1027
   ) -> (Pose) {
    alphaBoolBlend {blendOption=Linear}
    alphaScaleBiasClamp {bInterpResult=true, interpSpeedDecreasing=12, interpSpeedIncreasing=15}
    boneToModify {boneName=dyn_horn_r_05}
    constraintSetup [
     {targetBone {boneName=spine_05}}
     {targetBone {boneName=clavicle_r}}
     {targetBone {boneName=upperarm_twist_01_r}}
     {targetBone {boneName=neck_01}}
     {targetBone {boneName=head}}
     {targetBone {boneName=neck_02}}
    ]
    constraintWeights [1,1,1,1,1,1]
   }
   n10 = Anim Dynamics - Bone: dyn_horn_l_05(ComponentPose=@n9, Alpha=0.5, ExternalForce="0,0,0") {
    angularDampingOverride=1
    angularSpringConstant=20000
    bAngularSpring=true
    bOverrideAngularDamping=true
    bUseGravityOverride=true
    gravityScale=0.35
    simulationSpace=RootRelative
    alphaBoolBlend {blendOption=Linear}
    boundBone {boneName=dyn_horn_l_05}
    componentAppliedLinearAccClamp=<1500 1500 1500>
    componentLinearAccScale=<0.2 0.2 0.2>
    componentLinearVelScale=<0.2 0.2 0.2>
    gravityOverride {z=-980}
    physicsBodyDefinitions [{
      sphereCollisionRadius=0
      boundBone {boneName=dyn_horn_l_05}
      boxExtents=<20 30 30>
      constraintSetup {angularLimitsMax=<0 2 2>, angularLimitsMin=<0 -5 -10>, angularTarget=<1 0 0>}
      localJointOffset=<-4 0 0>
     }]
   }
   n11 = Anim Dynamics - Bone: dyn_horn_l_010(ComponentPose=@n10, Alpha=0.2, ExternalForce="0,0,0") {
    angularDampingOverride=1
    angularSpringConstant=20000
    bAngularSpring=true
    bOverrideAngularDamping=true
    bUseGravityOverride=true
    gravityScale=0.35
    simulationSpace=RootRelative
    alphaBoolBlend {blendOption=Linear}
    boundBone {boneName=dyn_horn_l_010}
    componentAppliedLinearAccClamp=<1500 1500 1500>
    componentLinearAccScale=<0.2 0.2 0.2>
    componentLinearVelScale=<0.2 0.2 0.2>
    gravityOverride {z=-980}
    physicsBodyDefinitions [{
      sphereCollisionRadius=0
      boundBone {boneName=dyn_horn_l_010}
      boxExtents=<20 30 30>
      constraintSetup {angularLimitsMax=<0 5 8>, angularLimitsMin=<0 -5 -8>, angularTarget=<1 0 0>}
      localJointOffset=<4 0 0>
     }]
   }
   n12 = Anim Dynamics - Bone: dyn_horn_r_05(ComponentPose=@n11, Alpha=0.5, ExternalForce="0,0,0") {
    angularDampingOverride=1
    angularSpringConstant=20000
    bAngularSpring=true
    bOverrideAngularDamping=true
    bUseGravityOverride=true
    gravityScale=0.35
    simulationSpace=RootRelative
    alphaBoolBlend {blendOption=Linear}
    boundBone {boneName=dyn_horn_r_05}
    componentAppliedLinearAccClamp=<1500 1500 1500>
    componentLinearAccScale=<0.2 0.2 0.2>
    componentLinearVelScale=<0.2 0.2 0.2>
    gravityOverride {z=-980}
    physicsBodyDefinitions [{
      sphereCollisionRadius=0
      boundBone {boneName=dyn_horn_r_05}
      boxExtents=<20 30 30>
      constraintSetup {angularLimitsMax=<0 5 10>, angularLimitsMin=<0 -2 -10>, angularTarget=<1 0 0>}
      localJointOffset=<4 0 0>
     }]
   }
   n13 = Anim Dynamics - Bone: dyn_horn_r_010(ComponentPose=@n12, Alpha=0.2, ExternalForce="0,0,0") {
    angularDampingOverride=1
    angularSpringConstant=20000
    bAngularSpring=true
    bOverrideAngularDamping=true
    bUseGravityOverride=true
    gravityScale=0.35
    simulationSpace=RootRelative
    alphaBoolBlend {blendOption=Linear}
    boundBone {boneName=dyn_horn_r_010}
    componentAppliedLinearAccClamp=<1500 1500 1500>
    componentLinearAccScale=<0.2 0.2 0.2>
    componentLinearVelScale=<0.2 0.2 0.2>
    gravityOverride {z=-980}
    physicsBodyDefinitions [{
      sphereCollisionRadius=0
      boundBone {boneName=dyn_horn_r_010}
      boxExtents=<20 30 30>
      constraintSetup {angularLimitsMax=<0 5 8>, angularLimitsMin=<0 -5 -8>, angularTarget=<1 0 0>}
      localJointOffset=<-4 0 0>
     }]
   }
   n14 = "Constraint - dyn_chain_fr_02 (P)"(ComponentPose=@n13, Alpha=1, ConstraintWeights_0=1) {
    alphaBoolBlend {blendOption=Linear}
    boneToModify {boneName=dyn_chain_fr_02}
    constraintSetup [{targetBone {boneName=dyn_chain_02}}]
    constraintWeights [1]
   }
   n15 = "Constraint - dyn_chain_bk_02 (P)"(ComponentPose=@n14, Alpha=1, ConstraintWeights_0=1) {
    alphaBoolBlend {blendOption=Linear}
    boneToModify {boneName=dyn_chain_bk_02}
    constraintSetup [{targetBone {boneName=dyn_chain_02}}]
    constraintWeights [1]
   }
   n16 = Anim Dynamics - Bone: dyn_chain_fr_01(ComponentPose=@n15, Alpha=1, ExternalForce="0,0,0") {
    angularDampingOverride=0.9999
    angularSpringConstant=500
    bAngularSpring=true
    bGravityOverrideInSimSpace=true
    bOverrideAngularDamping=true
    bUseSphericalLimits=true
    simulationSpace=BoneRelative
    alphaBoolBlend {blendOption=Linear}
    boundBone {boneName=dyn_chain_fr_01}
    componentAppliedLinearAccClamp=<3500 3500 3500>
    componentLinearAccScale=<1.2 1.2 1.2>
    componentLinearVelScale=<1.2 1.2 1.2>
    physicsBodyDefinitions [{
      sphereCollisionRadius=0
      boundBone {boneName=dyn_chain_fr_01}
      boxExtents=<10 30 30>
      constraintSetup {angularLimitsMax=<1 20 25>, angularLimitsMin=<-1 -40 -35>, angularTarget=<0 0 0>}
      localJointOffset=<8 0 0>
     }]
    relativeSpaceBone {boneName=root}
    sphericalLimits [
     {limitRadius=5, drivingBone {boneName=foot_r}, sphereLocalOffset=<0 0 0>}
     {limitRadius=5, drivingBone {boneName=calf_twist_01_r}, sphereLocalOffset=<0 0 0>}
    ]
   }
   n17 = Anim Dynamics - Bone: dyn_chain_fr_02(ComponentPose=@n16, Alpha=1, ExternalForce="0,0,0") {
    angularDampingOverride=0.9999
    angularSpringConstant=500
    bAngularSpring=true
    bGravityOverrideInSimSpace=true
    bOverrideAngularDamping=true
    bUseSphericalLimits=true
    simulationSpace=BoneRelative
    alphaBoolBlend {blendOption=Linear}
    boundBone {boneName=dyn_chain_fr_02}
    componentAppliedLinearAccClamp=<3500 3500 3500>
    componentLinearAccScale=<1.2 1.2 1.2>
    componentLinearVelScale=<1.2 1.2 1.2>
    physicsBodyDefinitions [{
      sphereCollisionRadius=0
      boundBone {boneName=dyn_chain_fr_02}
      boxExtents=<10 30 30>
      constraintSetup {angularLimitsMax=<1 20 15>, angularLimitsMin=<-1 -40 -15>, angularTarget=<0 0 0>}
      localJointOffset=<5 0 0>
     }]
    relativeSpaceBone {boneName=root}
    sphericalLimits [
     {limitRadius=5, drivingBone {boneName=foot_r}, sphereLocalOffset=<0 0 0>}
     {limitRadius=5, drivingBone {boneName=calf_twist_01_r}, sphereLocalOffset=<0 0 0>}
    ]
   }
   n18 = Anim Dynamics - Bone: dyn_chain_01(ComponentPose=@n17, Alpha=1, ExternalForce="0,0,0") {
    angularDampingOverride=0.9999
    angularSpringConstant=500
    bAngularSpring=true
    bGravityOverrideInSimSpace=true
    bOverrideAngularDamping=true
    bUseSphericalLimits=true
    simulationSpace=BoneRelative
    alphaBoolBlend {blendOption=Linear}
    boundBone {boneName=dyn_chain_01}
    componentAppliedLinearAccClamp=<3500 3500 3500>
    componentLinearAccScale=<1.2 1.2 1.2>
    componentLinearVelScale=<1.2 1.2 1.2>
    physicsBodyDefinitions [{
      sphereCollisionRadius=0
      boundBone {boneName=dyn_chain_01}
      boxExtents=<10 30 30>
      constraintSetup {angularLimitsMax=<1 45 25>, angularLimitsMin=<-1 -45 -35>, angularTarget=<0 0 0>}
      localJointOffset=<8 0 0>
     }]
    relativeSpaceBone {boneName=root}
    sphericalLimits [
     {limitRadius=5, drivingBone {boneName=foot_r}, sphereLocalOffset=<0 0 0>}
     {limitRadius=5, drivingBone {boneName=calf_twist_01_r}, sphereLocalOffset=<0 0 0>}
    ]
   }
   n19 = Anim Dynamics - Bone: dyn_chain_02(ComponentPose=@n18, Alpha=1, ExternalForce="0,0,0") {
    angularDampingOverride=0.9999
    angularSpringConstant=500
    bAngularSpring=true
    bGravityOverrideInSimSpace=true
    bOverrideAngularDamping=true
    bUseSphericalLimits=true
    simulationSpace=BoneRelative
    alphaBoolBlend {blendOption=Linear}
    boundBone {boneName=dyn_chain_02}
    componentAppliedLinearAccClamp=<3500 3500 3500>
    componentLinearAccScale=<1.2 1.2 1.2>
    componentLinearVelScale=<1.2 1.2 1.2>
    physicsBodyDefinitions [{
      sphereCollisionRadius=0
      boundBone {boneName=dyn_chain_02}
      boxExtents=<10 30 30>
      constraintSetup {angularLimitsMax=<1 45 15>, angularLimitsMin=<-1 -45 -15>, angularTarget=<0 0 0>}
      localJointOffset=<5 0 0>
     }]
    relativeSpaceBone {boneName=root}
    sphericalLimits [
     {limitRadius=5, drivingBone {boneName=foot_r}, sphereLocalOffset=<0 0 0>}
     {limitRadius=5, drivingBone {boneName=calf_twist_01_r}, sphereLocalOffset=<0 0 0>}
    ]
   }
   n20 = Anim Dynamics - Bone: dyn_chain_03(ComponentPose=@n19, Alpha=1, ExternalForce="0,0,0") {
    angularDampingOverride=0.9999
    angularSpringConstant=500
    bAngularSpring=true
    bGravityOverrideInSimSpace=true
    bOverrideAngularDamping=true
    bUseSphericalLimits=true
    simulationSpace=BoneRelative
    alphaBoolBlend {blendOption=Linear}
    boundBone {boneName=dyn_chain_03}
    componentAppliedLinearAccClamp=<3500 3500 3500>
    componentLinearAccScale=<1.2 1.2 1.2>
    componentLinearVelScale=<1.2 1.2 1.2>
    physicsBodyDefinitions [{
      sphereCollisionRadius=0
      boundBone {boneName=dyn_chain_03}
      boxExtents=<10 30 30>
      constraintSetup {angularLimitsMax=<1 45 15>, angularLimitsMin=<-1 -45 -15>, angularTarget=<0 0 0>}
      localJointOffset=<5 0 0>
     }]
    relativeSpaceBone {boneName=root}
    sphericalLimits [
     {limitRadius=5, drivingBone {boneName=foot_r}, sphereLocalOffset=<0 0 0>}
     {limitRadius=5, drivingBone {boneName=calf_twist_01_r}, sphereLocalOffset=<0 0 0>}
    ]
   }
   n21 = Anim Dynamics - Bone: dyn_chain_bk_01(ComponentPose=@n20, Alpha=1, ExternalForce="0,0,0") {
    angularDampingOverride=0.9999
    angularSpringConstant=500
    bAngularSpring=true
    bGravityOverrideInSimSpace=true
    bOverrideAngularDamping=true
    bUseSphericalLimits=true
    simulationSpace=BoneRelative
    alphaBoolBlend {blendOption=Linear}
    boundBone {boneName=dyn_chain_bk_01}
    componentAppliedLinearAccClamp=<3500 3500 3500>
    componentLinearAccScale=<1.2 1.2 1.2>
    componentLinearVelScale=<1.2 1.2 1.2>
    physicsBodyDefinitions [{
      sphereCollisionRadius=0
      boundBone {boneName=dyn_chain_bk_01}
      boxExtents=<10 30 30>
      constraintSetup {angularLimitsMax=<1 20 25>, angularLimitsMin=<-1 -35 -35>, angularTarget=<0 0 0>}
      localJointOffset=<8 0 0>
     }]
    relativeSpaceBone {boneName=root}
    sphericalLimits [
     {limitRadius=5, drivingBone {boneName=foot_r}, sphereLocalOffset=<0 0 0>}
     {limitRadius=5, drivingBone {boneName=calf_twist_01_r}, sphereLocalOffset=<0 0 0>}
    ]
   }
   n22 = Anim Dynamics - Bone: dyn_chain_bk_02(ComponentPose=@n21, Alpha=1, ExternalForce="0,0,0") {
    angularDampingOverride=0.9999
    angularSpringConstant=500
    bAngularSpring=true
    bGravityOverrideInSimSpace=true
    bOverrideAngularDamping=true
    bUseSphericalLimits=true
    simulationSpace=BoneRelative
    alphaBoolBlend {blendOption=Linear}
    boundBone {boneName=dyn_chain_bk_02}
    componentAppliedLinearAccClamp=<3500 3500 3500>
    componentLinearAccScale=<1.2 1.2 1.2>
    componentLinearVelScale=<1.2 1.2 1.2>
    physicsBodyDefinitions [{
      sphereCollisionRadius=0
      boundBone {boneName=dyn_chain_bk_02}
      boxExtents=<10 30 30>
      constraintSetup {angularLimitsMax=<1 20 15>, angularLimitsMin=<-1 -25 -15>, angularTarget=<0 0 0>}
      localJointOffset=<5 0 0>
     }]
    relativeSpaceBone {boneName=root}
    sphericalLimits [
     {limitRadius=5, drivingBone {boneName=foot_r}, sphereLocalOffset=<0 0 0>}
     {limitRadius=5, drivingBone {boneName=calf_twist_01_r}, sphereLocalOffset=<0 0 0>}
    ]
   }
   n23 = "Constraint - dyn_neck_ring (PPP)"!(ComponentPose=@n22, Alpha=1, ConstraintWeights_0=0.9721, ConstraintWeights_1:float=0.0144, ConstraintWeights_2:float=0.0134) -> (Pose) {
    alphaBoolBlend {blendOption=Linear}
    boneToModify {boneName=dyn_neck_ring}
    constraintSetup [{targetBone {boneName=spine_05}}, {targetBone {boneName=clavicle_l}}, {targetBone {boneName=clavicle_r}}]
    constraintWeights [1,1,1]
   }
   n24 = Anim Dynamics - Bone: dyn_neck_ring(ComponentPose=@n23, Alpha=0.5, ExternalForce="0,0,0") {
    angularDampingOverride=0.9999
    angularSpringConstant=500
    bAngularSpring=true
    bGravityOverrideInSimSpace=true
    bOverrideAngularDamping=true
    bUseGravityOverride=true
    simulationSpace=BoneRelative
    alphaBoolBlend {blendOption=Linear}
    boundBone {boneName=dyn_neck_ring}
    componentAppliedLinearAccClamp=<3500 3500 3500>
    componentLinearAccScale=<0.2 0.3 0.3>
    componentLinearVelScale=<0.2 0.3 0.3>
    gravityOverride {z=-980}
    physicsBodyDefinitions [{
      sphereCollisionRadius=0
      boundBone {boneName=dyn_neck_ring}
      boxExtents=<10 30 30>
      constraintSetup {angularLimitsMax=<1 5 1>, angularLimitsMin=<-1 -5 -2>, angularTarget=<0 0 0>}
      localJointOffset=<2 0 0>
     }]
    relativeSpaceBone {boneName=spine_05}
   }
   n25 = "Transform (Modify) Bone - Bone: dyn_neck_ring"(ComponentPose=@n24, Alpha=1, Translation="0,0,0", Rotation="0,2,0", Scale="1,1,1") {
    rotationMode=BMM_Additive
    rotationSpace=BCS_BoneSpace
    alphaBoolBlend {blendOption=Linear}
    boneToModify {boneName=dyn_neck_ring}
    rotation {yaw=2}
   }
   n26 = Anim Dynamics - Bone: dyn_bow_in_01(ComponentPose=@n25, Alpha=1, ExternalForce="0,0,0") {
    angularDampingOverride=0.9999
    angularSpringConstant=500
    bAngularSpring=true
    bGravityOverrideInSimSpace=true
    bOverrideAngularDamping=true
    simulationSpace=BoneRelative
    alphaBoolBlend {blendOption=Linear}
    boundBone {boneName=dyn_bow_in_01}
    componentAppliedLinearAccClamp=<3500 3500 3500>
    componentLinearAccScale=<1.2 1.2 1.2>
    componentLinearVelScale=<1.2 1.2 1.2>
    physicsBodyDefinitions [{
      sphereCollisionRadius=0
      boundBone {boneName=dyn_bow_in_01}
      boxExtents=<15 20 20>
      constraintSetup {angularLimitsMax=<1 5 20>, angularLimitsMin=<-1 -45 -10>, angularTarget=<0 0 0>}
      localJointOffset=<3 0 0>
     }]
    relativeSpaceBone {boneName=root}
   }
   n27 = Anim Dynamics - Bone: dyn_bow_in_02(ComponentPose=@n26, Alpha=1, ExternalForce="0,0,0") {
    angularDampingOverride=0.9999
    angularSpringConstant=500
    bAngularSpring=true
    bGravityOverrideInSimSpace=true
    bOverrideAngularDamping=true
    simulationSpace=BoneRelative
    alphaBoolBlend {blendOption=Linear}
    boundBone {boneName=dyn_bow_in_02}
    componentAppliedLinearAccClamp=<3500 3500 3500>
    componentLinearAccScale=<1.2 1.2 1.2>
    componentLinearVelScale=<1.2 1.2 1.2>
    physicsBodyDefinitions [{
      sphereCollisionRadius=0
      boundBone {boneName=dyn_bow_in_02}
      boxExtents=<15 20 20>
      constraintSetup {angularLimitsMax=<1 8 40>, angularLimitsMin=<-1 -45 -10>, angularTarget=<0 0 0>}
      localJointOffset=<3 0 0>
     }]
    relativeSpaceBone {boneName=root}
   }
   n28 = Anim Dynamics - Bone: dyn_bow_in_03(ComponentPose=@n27, Alpha=1, ExternalForce="0,0,0") {
    angularDampingOverride=0.9999
    angularSpringConstant=500
    bAngularSpring=true
    bGravityOverrideInSimSpace=true
    bOverrideAngularDamping=true
    simulationSpace=BoneRelative
    alphaBoolBlend {blendOption=Linear}
    boundBone {boneName=dyn_bow_in_03}
    componentAppliedLinearAccClamp=<3500 3500 3500>
    componentLinearAccScale=<1.2 1.2 1.2>
    componentLinearVelScale=<1.2 1.2 1.2>
    physicsBodyDefinitions [{
      sphereCollisionRadius=0
      boundBone {boneName=dyn_bow_in_03}
      boxExtents=<15 20 20>
      constraintSetup {angularLimitsMax=<1 8 20>, angularLimitsMin=<-1 -45 -10>, angularTarget=<0 0 0>}
      localJointOffset=<3 0 0>
     }]
    relativeSpaceBone {boneName=root}
   }
   n29 = Anim Dynamics - Bone: dyn_lace_in_01(ComponentPose=@n28, Alpha=1, ExternalForce="0,0,0") {
    angularDampingOverride=0.9999
    angularSpringConstant=500
    bAngularSpring=true
    bGravityOverrideInSimSpace=true
    bOverrideAngularDamping=true
    simulationSpace=BoneRelative
    alphaBoolBlend {blendOption=Linear}
    boundBone {boneName=dyn_lace_in_01}
    componentAppliedLinearAccClamp=<3500 3500 3500>
    componentLinearAccScale=<1.2 1.2 1.2>
    componentLinearVelScale=<1.2 1.2 1.2>
    physicsBodyDefinitions [{
      sphereCollisionRadius=0
      boundBone {boneName=dyn_lace_in_01}
      boxExtents=<15 20 20>
      constraintSetup {angularLimitsMax=<1 5 20>, angularLimitsMin=<-1 -45 -10>, angularTarget=<0 0 0>}
      localJointOffset=<3 0 0>
     }]
    relativeSpaceBone {boneName=root}
   }
   n30 = Anim Dynamics - Bone: dyn_lace_in_02(ComponentPose=@n29, Alpha=1, ExternalForce="0,0,0") {
    angularDampingOverride=0.9999
    angularSpringConstant=500
    bAngularSpring=true
    bGravityOverrideInSimSpace=true
    bOverrideAngularDamping=true
    simulationSpace=BoneRelative
    alphaBoolBlend {blendOption=Linear}
    boundBone {boneName=dyn_lace_in_02}
    componentAppliedLinearAccClamp=<3500 3500 3500>
    componentLinearAccScale=<1.2 1.2 1.2>
    componentLinearVelScale=<1.2 1.2 1.2>
    physicsBodyDefinitions [{
      sphereCollisionRadius=0
      boundBone {boneName=dyn_lace_in_02}
      boxExtents=<15 20 20>
      constraintSetup {angularLimitsMax=<1 8 40>, angularLimitsMin=<-1 -45 -10>, angularTarget=<0 0 0>}
      localJointOffset=<3 0 0>
     }]
    relativeSpaceBone {boneName=root}
   }
   n31 = Anim Dynamics - Bone: dyn_lace_in_03(ComponentPose=@n30, Alpha=1, ExternalForce="0,0,0") {
    angularDampingOverride=0.9999
    angularSpringConstant=500
    bAngularSpring=true
    bGravityOverrideInSimSpace=true
    bOverrideAngularDamping=true
    simulationSpace=BoneRelative
    alphaBoolBlend {blendOption=Linear}
    boundBone {boneName=dyn_lace_in_03}
    componentAppliedLinearAccClamp=<3500 3500 3500>
    componentLinearAccScale=<1.2 1.2 1.2>
    componentLinearVelScale=<1.2 1.2 1.2>
    physicsBodyDefinitions [{
      sphereCollisionRadius=0
      boundBone {boneName=dyn_lace_in_03}
      boxExtents=<15 20 20>
      constraintSetup {angularLimitsMax=<1 8 20>, angularLimitsMin=<-1 -45 -10>, angularTarget=<0 0 0>}
      localJointOffset=<3 0 0>
     }]
    relativeSpaceBone {boneName=root}
   }
   n32 = Anim Dynamics - Bone: dyn_bow_out_01(ComponentPose=@n31, Alpha=1, ExternalForce="0,0,0") {
    angularDampingOverride=0.9999
    angularSpringConstant=500
    bAngularSpring=true
    bGravityOverrideInSimSpace=true
    bOverrideAngularDamping=true
    simulationSpace=BoneRelative
    alphaBoolBlend {blendOption=Linear}
    boundBone {boneName=dyn_bow_out_01}
    componentAppliedLinearAccClamp=<5000 5000 5000>
    componentLinearAccScale=<1.2 1.2 1.2>
    componentLinearVelScale=<1.2 1.2 1.2>
    physicsBodyDefinitions [{
      sphereCollisionRadius=0
      boundBone {boneName=dyn_bow_out_01}
      boxExtents=<15 20 20>
      constraintSetup {angularLimitsMax=<1 5 20>, angularLimitsMin=<-1 -45 -10>, angularTarget=<0 0 0>}
      localJointOffset=<3 0 0>
     }]
    relativeSpaceBone {boneName=root}
   }
   n33 = Anim Dynamics - Bone: dyn_bow_out_02(ComponentPose=@n32, Alpha=1, ExternalForce="0,0,0") {
    angularDampingOverride=0.9999
    angularSpringConstant=500
    bAngularSpring=true
    bGravityOverrideInSimSpace=true
    bOverrideAngularDamping=true
    simulationSpace=BoneRelative
    alphaBoolBlend {blendOption=Linear}
    boundBone {boneName=dyn_bow_out_02}
    componentAppliedLinearAccClamp=<5000 5000 5000>
    componentLinearAccScale=<1.2 1.2 1.2>
    componentLinearVelScale=<1.2 1.2 1.2>
    physicsBodyDefinitions [{
      sphereCollisionRadius=0
      boundBone {boneName=dyn_bow_out_02}
      boxExtents=<15 20 20>
      constraintSetup {angularLimitsMax=<1 8 40>, angularLimitsMin=<-1 -45 -10>, angularTarget=<0 0 0>}
      localJointOffset=<3 0 0>
     }]
    relativeSpaceBone {boneName=root}
   }
   n34 = Anim Dynamics - Bone: dyn_bow_out_03(ComponentPose=@n33, Alpha=1, ExternalForce="0,0,0") {
    angularDampingOverride=0.9999
    angularSpringConstant=500
    bAngularSpring=true
    bGravityOverrideInSimSpace=true
    bOverrideAngularDamping=true
    simulationSpace=BoneRelative
    alphaBoolBlend {blendOption=Linear}
    boundBone {boneName=dyn_bow_out_03}
    componentAppliedLinearAccClamp=<5000 5000 5000>
    componentLinearAccScale=<1.2 1.2 1.2>
    componentLinearVelScale=<1.2 1.2 1.2>
    physicsBodyDefinitions [{
      sphereCollisionRadius=0
      boundBone {boneName=dyn_bow_out_03}
      boxExtents=<15 20 20>
      constraintSetup {angularLimitsMax=<1 8 20>, angularLimitsMin=<-1 -45 -10>, angularTarget=<0 0 0>}
      localJointOffset=<3 0 0>
     }]
    relativeSpaceBone {boneName=root}
   }
   n35 = Anim Dynamics - Bone: dyn_lace_out_01(ComponentPose=@n34, Alpha=1, ExternalForce="0,0,0") {
    angularDampingOverride=0.9999
    angularSpringConstant=500
    bAngularSpring=true
    bGravityOverrideInSimSpace=true
    bOverrideAngularDamping=true
    simulationSpace=BoneRelative
    alphaBoolBlend {blendOption=Linear}
    boundBone {boneName=dyn_lace_out_01}
    componentAppliedLinearAccClamp=<5000 5000 5000>
    componentLinearAccScale=<1.2 1.2 1.2>
    componentLinearVelScale=<1.2 1.2 1.2>
    physicsBodyDefinitions [{
      sphereCollisionRadius=0
      boundBone {boneName=dyn_lace_out_01}
      boxExtents=<15 20 20>
      constraintSetup {angularLimitsMax=<1 5 20>, angularLimitsMin=<-1 -45 -10>, angularTarget=<0 0 0>}
      localJointOffset=<3 0 0>
     }]
    relativeSpaceBone {boneName=root}
   }
   n36 = Anim Dynamics - Bone: dyn_lace_out_02(ComponentPose=@n35, Alpha=1, ExternalForce="0,0,0") {
    angularDampingOverride=0.9999
    angularSpringConstant=500
    bAngularSpring=true
    bGravityOverrideInSimSpace=true
    bOverrideAngularDamping=true
    simulationSpace=BoneRelative
    alphaBoolBlend {blendOption=Linear}
    boundBone {boneName=dyn_lace_out_02}
    componentAppliedLinearAccClamp=<5000 5000 5000>
    componentLinearAccScale=<1.2 1.2 1.2>
    componentLinearVelScale=<1.2 1.2 1.2>
    physicsBodyDefinitions [{
      sphereCollisionRadius=0
      boundBone {boneName=dyn_lace_out_02}
      boxExtents=<15 20 20>
      constraintSetup {angularLimitsMax=<1 8 40>, angularLimitsMin=<-1 -45 -10>, angularTarget=<0 0 0>}
      localJointOffset=<3 0 0>
     }]
    relativeSpaceBone {boneName=root}
   }
   n37 = Anim Dynamics - Bone: dyn_lace_out_03(ComponentPose=@n36, Alpha=1, ExternalForce="0,0,0") {
    angularDampingOverride=0.9999
    angularSpringConstant=500
    bAngularSpring=true
    bGravityOverrideInSimSpace=true
    bOverrideAngularDamping=true
    simulationSpace=BoneRelative
    alphaBoolBlend {blendOption=Linear}
    boundBone {boneName=dyn_lace_out_03}
    componentAppliedLinearAccClamp=<5000 5000 5000>
    componentLinearAccScale=<1.2 1.2 1.2>
    componentLinearVelScale=<1.2 1.2 1.2>
    physicsBodyDefinitions [{
      sphereCollisionRadius=0
      boundBone {boneName=dyn_lace_out_03}
      boxExtents=<15 20 20>
      constraintSetup {angularLimitsMax=<1 8 20>, angularLimitsMin=<-1 -45 -10>, angularTarget=<0 0 0>}
      localJointOffset=<3 0 0>
     }]
    relativeSpaceBone {boneName=root}
   }
   n38 = "Constraint - dyn_horn_l_05 (PPPPPP)"(
    ComponentPose=@n37
    Alpha=0.5
    ConstraintWeights_0=0.2036
    ConstraintWeights_1=0.0842
    ConstraintWeights_2=0.0645
    ConstraintWeights_3=0.1476
    ConstraintWeights_4=0.25
    ConstraintWeights_5=0.25
   ) {
    alphaBoolBlend {blendOption=Linear}
    alphaScaleBiasClamp {bInterpResult=true, interpSpeedDecreasing=5, interpSpeedIncreasing=7}
    boneToModify {boneName=dyn_horn_l_05}
    constraintSetup [
     {targetBone {boneName=spine_05}}
     {targetBone {boneName=neck_01}}
     {targetBone {boneName=head}}
     {targetBone {boneName=neck_02}}
     {targetBone {boneName=dyn_shoulder_01}}
     {targetBone {boneName=dyn_shoulder_02}}
    ]
    constraintWeights [1,1,1,1,1,1]
   }
   n39 = "Constraint - dyn_horn_l_06 (PPPPPP)"!(
    ComponentPose=@n38
    Alpha=0.7
    ConstraintWeights_0=0.2036
    ConstraintWeights_1:float=0.0842
    ConstraintWeights_2:float=0.0645
    ConstraintWeights_3:float=0.1476
    ConstraintWeights_4:float=0.25
    ConstraintWeights_5:float=0.25
   ) -> (Pose) {
    alphaBoolBlend {blendOption=Linear}
    alphaScaleBiasClamp {bInterpResult=true, interpSpeedDecreasing=5, interpSpeedIncreasing=7}
    boneToModify {boneName=dyn_horn_l_06}
    constraintSetup [
     {targetBone {boneName=spine_05}}
     {targetBone {boneName=neck_01}}
     {targetBone {boneName=head}}
     {targetBone {boneName=neck_02}}
     {targetBone {boneName=dyn_shoulder_01}}
     {targetBone {boneName=dyn_shoulder_02}}
    ]
    constraintWeights [1,1,1,1,1,1]
   }
   n40 = "Constraint - dyn_horn_l_07 (PPPPPP)"!(
    ComponentPose=@n39
    Alpha=0.5
    ConstraintWeights_0=0.2036
    ConstraintWeights_1:float=0.0842
    ConstraintWeights_2:float=0.0645
    ConstraintWeights_3:float=0.1476
    ConstraintWeights_4:float=0.25
    ConstraintWeights_5:float=0.25
   ) -> (Pose) {
    alphaBoolBlend {blendOption=Linear}
    alphaScaleBiasClamp {bInterpResult=true, interpSpeedDecreasing=5, interpSpeedIncreasing=7}
    boneToModify {boneName=dyn_horn_l_07}
    constraintSetup [
     {targetBone {boneName=spine_05}}
     {targetBone {boneName=neck_01}}
     {targetBone {boneName=head}}
     {targetBone {boneName=neck_02}}
     {targetBone {boneName=dyn_shoulder_01}}
     {targetBone {boneName=dyn_shoulder_02}}
    ]
    constraintWeights [1,1,1,1,1,1]
   }
   n41 = "Constraint - dyn_horn_r_05 (PPPPPP)"(
    ComponentPose=@n40
    Alpha=0.5
    ConstraintWeights_0=0.2995
    ConstraintWeights_1=0.2888
    ConstraintWeights_2=0.2075
    ConstraintWeights_3=0.0684
    ConstraintWeights_4=0.033
    ConstraintWeights_5=0.1027
   ) {
    alphaBoolBlend {blendOption=Linear}
    alphaScaleBiasClamp {bInterpResult=true, interpSpeedDecreasing=5, interpSpeedIncreasing=7}
    boneToModify {boneName=dyn_horn_r_05}
    constraintSetup [
     {targetBone {boneName=spine_05}}
     {targetBone {boneName=clavicle_r}}
     {targetBone {boneName=upperarm_twist_01_r}}
     {targetBone {boneName=neck_01}}
     {targetBone {boneName=head}}
     {targetBone {boneName=neck_02}}
    ]
    constraintWeights [1,1,1,1,1,1]
   }
   n42 = "Constraint - dyn_horn_r_06 (PPPPPP)"!(
    ComponentPose=@n41
    Alpha=0.7
    ConstraintWeights_0=0.2995
    ConstraintWeights_1:float=0.2888
    ConstraintWeights_2:float=0.2075
    ConstraintWeights_3:float=0.0684
    ConstraintWeights_4:float=0.033
    ConstraintWeights_5:float=0.1027
   ) -> (Pose) {
    alphaBoolBlend {blendOption=Linear}
    alphaScaleBiasClamp {bInterpResult=true, interpSpeedDecreasing=5, interpSpeedIncreasing=7}
    boneToModify {boneName=dyn_horn_r_06}
    constraintSetup [
     {targetBone {boneName=spine_05}}
     {targetBone {boneName=clavicle_r}}
     {targetBone {boneName=upperarm_twist_01_r}}
     {targetBone {boneName=neck_01}}
     {targetBone {boneName=head}}
     {targetBone {boneName=neck_02}}
    ]
    constraintWeights [1,1,1,1,1,1]
   }
   n43 = "Constraint - dyn_horn_r_07 (PPPPPP)"!(
    ComponentPose=@n42
    Alpha=0.5
    ConstraintWeights_0=0.2995
    ConstraintWeights_1:float=0.2888
    ConstraintWeights_2:float=0.2075
    ConstraintWeights_3:float=0.0684
    ConstraintWeights_4:float=0.033
    ConstraintWeights_5:float=0.1027
   ) -> (Pose) {
    alphaBoolBlend {blendOption=Linear}
    alphaScaleBiasClamp {bInterpResult=true, interpSpeedDecreasing=5, interpSpeedIncreasing=7}
    boneToModify {boneName=dyn_horn_r_07}
    constraintSetup [
     {targetBone {boneName=spine_05}}
     {targetBone {boneName=clavicle_r}}
     {targetBone {boneName=upperarm_twist_01_r}}
     {targetBone {boneName=neck_01}}
     {targetBone {boneName=head}}
     {targetBone {boneName=neck_02}}
    ]
    constraintWeights [1,1,1,1,1,1]
   }
   n44 = "Constraint - dyn_horn_l_08 (PPPP)"!(
    ComponentPose=@n43
    Alpha=1
    ConstraintWeights_0=0.0022
    ConstraintWeights_1:float=0.1888
    ConstraintWeights_2:float=0.594
    ConstraintWeights_3:float=0.215
   ) -> (Pose) {
    alphaBoolBlend {blendOption=Linear}
    alphaScaleBiasClamp {bInterpResult=true, interpSpeedDecreasing=5, interpSpeedIncreasing=7}
    boneToModify {boneName=dyn_horn_l_08}
    constraintSetup [
     {targetBone {boneName=upperarm_twist_01_l}}
     {targetBone {boneName=spine_04}}
     {targetBone {boneName=spine_05}}
     {targetBone {boneName=clavicle_l}}
    ]
    constraintWeights [1,1,1,1]
   }
   n45 = "Constraint - dyn_horn_r_08 (PPPP)"!(
    ComponentPose=@n44
    Alpha=1
    ConstraintWeights_0=0.1617
    ConstraintWeights_1:float=0.6003
    ConstraintWeights_2:float=0.2339
    ConstraintWeights_3:float=0.0041
   ) -> (Pose) {
    alphaBoolBlend {blendOption=Linear}
    alphaScaleBiasClamp {bInterpResult=true, interpSpeedDecreasing=5, interpSpeedIncreasing=7}
    boneToModify {boneName=dyn_horn_r_08}
    constraintSetup [
     {targetBone {boneName=spine_04}}
     {targetBone {boneName=spine_05}}
     {targetBone {boneName=clavicle_r}}
     {targetBone {boneName=upperarm_twist_01_r}}
    ]
    constraintWeights [1,1,1,1]
   }
   n46 = "Constraint - dyn_horn_l_010 (PP)"!(ComponentPose=@n45, Alpha=0.8, ConstraintWeights_0=0.7718, ConstraintWeights_1:float=0.2282) -> (Pose) {
    alphaBoolBlend {blendOption=Linear}
    alphaScaleBiasClamp {bInterpResult=true, interpSpeedDecreasing=5, interpSpeedIncreasing=7}
    boneToModify {boneName=dyn_horn_l_010}
    constraintSetup [{targetBone {boneName=spine_04}}, {targetBone {boneName=spine_05}}]
    constraintWeights [1,1]
   }
   n47 = "Constraint - dyn_horn_r_010 (PPP)"!(ComponentPose=@n46, Alpha=1, ConstraintWeights_0=0.7393, ConstraintWeights_1:float=0.2579, ConstraintWeights_2:float=0.0027) -> (Pose) {
    alphaBoolBlend {blendOption=Linear}
    alphaScaleBiasClamp {bInterpResult=true, interpSpeedDecreasing=5, interpSpeedIncreasing=7}
    boneToModify {boneName=dyn_horn_r_010}
    constraintSetup [{targetBone {boneName=spine_04}}, {targetBone {boneName=spine_05}}, {targetBone {boneName=clavicle_r}}]
    constraintWeights [1,1,1]
   }
   n48 = "Constraint - dyn_strap_fr_Top (PP)"(ComponentPose=@n47, Alpha=1, ConstraintWeights_0=0.5, ConstraintWeights_1=0.5) {
    alphaBoolBlend {blendOption=Linear}
    boneToModify {boneName=dyn_strap_fr_Top}
    constraintSetup [{targetBone {boneName=dyn_shoulder_01}}, {targetBone {boneName=dyn_shoulder_02}}]
    constraintWeights [1,1]
   }
   n49 = "Constraint - dyn_strap_fr_C (PPPP)"(ComponentPose=@n48, Alpha=1, ConstraintWeights_0=0.4472, ConstraintWeights_1=0.0501, ConstraintWeights_2=0.0957, ConstraintWeights_3=0.407) {
    alphaBoolBlend {blendOption=Linear}
    boneToModify {boneName=dyn_strap_fr_C}
    constraintSetup [
     {targetBone {boneName=upperarm_twist_01_l}}
     {targetBone {boneName=spine_04}}
     {targetBone {boneName=spine_05}}
     {targetBone {boneName=clavicle_l}}
    ]
    constraintWeights [1,1,1,1]
   }
   n50 = "Constraint - dyn_strap_fr_Bot (PPPP)"(ComponentPose=@n49, Alpha=1, ConstraintWeights_0=0.1188, ConstraintWeights_1=0.2131, ConstraintWeights_2=0.2547, ConstraintWeights_3=0.4134) {
    alphaBoolBlend {blendOption=Linear}
    boneToModify {boneName=dyn_strap_fr_Bot}
    constraintSetup [
     {targetBone {boneName=upperarm_twist_01_l}}
     {targetBone {boneName=spine_04}}
     {targetBone {boneName=spine_05}}
     {targetBone {boneName=clavicle_l}}
    ]
    constraintWeights [1,1,1,1]
   }
   n51 = "Constraint - dyn_strap_bk_Top (PP)"(ComponentPose=@n50, Alpha=1, ConstraintWeights_0=0.5, ConstraintWeights_1=0.5) {
    alphaBoolBlend {blendOption=Linear}
    boneToModify {boneName=dyn_strap_bk_Top}
    constraintSetup [{targetBone {boneName=dyn_shoulder_01}}, {targetBone {boneName=dyn_shoulder_02}}]
    constraintWeights [1,1]
   }
   n52 = "Constraint - dyn_strap_bk_C (PPP)"(ComponentPose=@n51, Alpha=1, ConstraintWeights_0=0.2509, ConstraintWeights_1=0.031, ConstraintWeights_2=0.7181) {
    alphaBoolBlend {blendOption=Linear}
    boneToModify {boneName=dyn_strap_bk_C}
    constraintSetup [{targetBone {boneName=upperarm_twist_01_l}}, {targetBone {boneName=spine_05}}, {targetBone {boneName=clavicle_l}}]
    constraintWeights [1,1,1]
   }
   n53 = "Constraint - dyn_strap_bk_Bot (PPPPPPP)"(
    ComponentPose=@n52
    Alpha=1
    ConstraintWeights_0=0.0952
    ConstraintWeights_1=0
    ConstraintWeights_2=0.0347
    ConstraintWeights_3=0
    ConstraintWeights_4=0.1824
    ConstraintWeights_5=0.6863
    ConstraintWeights_6=0.0014
   ) {
    alphaBoolBlend {blendOption=Linear}
    boneToModify {boneName=dyn_strap_bk_Bot}
    constraintSetup [
     {targetBone {boneName=upperarm_twist_01_l}}
     {targetBone {boneName=spine_02}}
     {targetBone {boneName=spine_04}}
     {targetBone {boneName=spine_03}}
     {targetBone {boneName=spine_05}}
     {targetBone {boneName=clavicle_l}}
     {targetBone {boneName=clavicle_r}}
    ]
    constraintWeights [1,1,1,1,1,1,1]
   }
   n54 = "Transform (Modify) Bone - Bone: dyn_horn_l_07"!(ComponentPose=@n53, Alpha=Get Jump() -> (Jump:double), Rotation="0,-15,0") -> (Pose) {
    rotationMode=BMM_Additive
    rotationSpace=BCS_BoneSpace
    alphaBoolBlend {blendOption=Linear}
    boneToModify {boneName=dyn_horn_l_07}
    rotation {yaw=-15}
   }
   n55 = "Constraint - dyn_strap_fr_Top (PP)"(ComponentPose=@n54, Alpha=1, ConstraintWeights_0=0.552, ConstraintWeights_1=0.448) {
    alphaBoolBlend {blendOption=Linear}
    boneToModify {boneName=dyn_strap_fr_Top}
    constraintSetup [{targetBone {boneName=dyn_shoulder_01}}, {targetBone {boneName=dyn_shoulder_02}}]
    constraintWeights [1,1]
   }
   n56 = "Constraint - dyn_strap_fr_C (PPPP)"(ComponentPose=@n55, Alpha=1, ConstraintWeights_0=0.5074, ConstraintWeights_1=0.0392, ConstraintWeights_2=0.0823, ConstraintWeights_3=0.3711) {
    alphaBoolBlend {blendOption=Linear}
    boneToModify {boneName=dyn_strap_fr_C}
    constraintSetup [
     {targetBone {boneName=upperarm_twist_01_l}}
     {targetBone {boneName=spine_04}}
     {targetBone {boneName=spine_05}}
     {targetBone {boneName=clavicle_l}}
    ]
    constraintWeights [1,1,1,1]
   }
   n57 = "Constraint - dyn_strap_fr_Bot (PPPP)"(ComponentPose=@n56, Alpha=1, ConstraintWeights_0=0.1188, ConstraintWeights_1=0.2131, ConstraintWeights_2=0.2547, ConstraintWeights_3=0.4134) {
    alphaBoolBlend {blendOption=Linear}
    boneToModify {boneName=dyn_strap_fr_Bot}
    constraintSetup [
     {targetBone {boneName=upperarm_twist_01_l}}
     {targetBone {boneName=spine_04}}
     {targetBone {boneName=spine_05}}
     {targetBone {boneName=clavicle_l}}
    ]
    constraintWeights [1,1,1,1]
   }
   n58 = "Constraint - dyn_strap_bk_Top (PP)"(ComponentPose=@n57, Alpha=1, ConstraintWeights_0=0.7733, ConstraintWeights_1=0.2267) {
    alphaBoolBlend {blendOption=Linear}
    boneToModify {boneName=dyn_strap_bk_Top}
    constraintSetup [{targetBone {boneName=upperarm_twist_01_l}}, {targetBone {boneName=clavicle_l}}]
    constraintWeights [1,1]
   }
   n59 = "Constraint - dyn_strap_bk_C (PPPP)"!(
    ComponentPose=@n58
    Alpha=1
    ConstraintWeights_0=0.5178
    ConstraintWeights_1:float=0
    ConstraintWeights_2:float=0.0155
    ConstraintWeights_3:float=0.4667
   ) -> (Pose) {
    alphaBoolBlend {blendOption=Linear}
    boneToModify {boneName=dyn_strap_bk_C}
    constraintSetup [
     {targetBone {boneName=upperarm_twist_01_l}}
     {targetBone {boneName=spine_04}}
     {targetBone {boneName=spine_05}}
     {targetBone {boneName=clavicle_l}}
    ]
    constraintWeights [1,1,1,1]
   }
   n60 = "Constraint - dyn_strap_bk_Bot (PPPPP)"!(
    ComponentPose=@n59
    Alpha=1
    ConstraintWeights_0=0.1245
    ConstraintWeights_1:float=0.0175
    ConstraintWeights_2:float=0.214
    ConstraintWeights_3:float=0.644
    ConstraintWeights_4:float=0
   ) -> (Pose) {
    alphaBoolBlend {blendOption=Linear}
    boneToModify {boneName=dyn_strap_bk_Bot}
    constraintSetup [
     {targetBone {boneName=upperarm_twist_01_l}}
     {targetBone {boneName=spine_04}}
     {targetBone {boneName=spine_05}}
     {targetBone {boneName=clavicle_l}}
     {targetBone {boneName=clavicle_r}}
    ]
    constraintWeights [1,1,1,1,1]
   }
   n61 = "Transform (Modify) Bone - Bone: dyn_strap_fr_Top"!(ComponentPose=@n60, Alpha=1, Translation="-1.836,-1.660,0.339", Rotation="-14.381,55.181,5.089") -> (Pose) {
    rotationMode=BMM_Additive
    rotationSpace=BCS_BoneSpace
    translationMode=BMM_Additive
    translationSpace=BCS_BoneSpace
    alphaBoolBlend {blendOption=Linear}
    boneToModify {boneName=dyn_strap_fr_Top}
    rotation=<rot -14.381 55.181 5.089>
    translation=<-1.836 -1.66 0.339>
   }
   n62 = "Transform (Modify) Bone - Bone: dyn_strap_fr_C"!(ComponentPose=@n61, Alpha=1, Translation="-0.000,0.464,0.082", Rotation="-0.000,0.000,-10.000") -> (Pose) {
    rotationMode=BMM_Additive
    rotationSpace=BCS_BoneSpace
    translationMode=BMM_Additive
    translationSpace=BCS_BoneSpace
    alphaBoolBlend {blendOption=Linear}
    boneToModify {boneName=dyn_strap_fr_C}
    rotation=<rot 0 0.0001 -10>
    translation {y=0.464, z=0.082}
   }
   n63 = "Transform (Modify) Bone - Bone: dyn_strap_fr_Bot"!(ComponentPose=@n62, Alpha=1, Translation="0.497,-0.765,-0.000", Rotation="-0.000,0.000,-5.000") -> (Pose) {
    rotationMode=BMM_Additive
    rotationSpace=BCS_BoneSpace
    translationMode=BMM_Additive
    translationSpace=BCS_BoneSpace
    alphaBoolBlend {blendOption=Linear}
    boneToModify {boneName=dyn_strap_fr_Bot}
    rotation=<rot 0 0 -5>
    translation=<0.497 -0.765>
   }
   n64 = "Transform (Modify) Bone - Bone: dyn_strap_bk_Top"!(ComponentPose=@n63, Alpha=1, Translation="1.578,-2.242,9.070", Rotation="-49.243,-18.916,0.995") -> (Pose) {
    rotationMode=BMM_Additive
    rotationSpace=BCS_BoneSpace
    translationMode=BMM_Additive
    translationSpace=BCS_BoneSpace
    alphaBoolBlend {blendOption=Linear}
    boneToModify {boneName=dyn_strap_bk_Top}
    rotation=<rot -49.243 -18.916 0.995>
    translation=<1.578 -2.242 9.07>
   }
   n65 = "Transform (Modify) Bone - Bone: dyn_strap_bk_C"!(ComponentPose=@n64, Alpha=1, Translation="0.059,0.000,0.653", Rotation="-5.000,0.000,15.000") -> (Pose) {
    rotationMode=BMM_Additive
    rotationSpace=BCS_BoneSpace
    translationMode=BMM_Additive
    translationSpace=BCS_BoneSpace
    alphaBoolBlend {blendOption=Linear}
    boneToModify {boneName=dyn_strap_bk_C}
    rotation=<rot -5 0 15>
    translation {x=0.059, z=0.653}
   }
   n66 = "Transform (Modify) Bone - Bone: dyn_strap_bk_Bot"!(ComponentPose=@n65, Alpha=1, Translation="1.384,-0.079,1.785", Rotation="0,0,0") -> (Pose) {
    rotationMode=BMM_Additive
    rotationSpace=BCS_BoneSpace
    translationMode=BMM_Additive
    translationSpace=BCS_BoneSpace
    alphaBoolBlend {blendOption=Linear}
    boneToModify {boneName=dyn_strap_bk_Bot}
    rotation=<rot 0 0 0>
    translation=<1.384 -0.079 1.785>
   }
   n67 = "Transform (Modify) Bone - Bone: dyn_strap_fr_Top"(ComponentPose=@n66, Alpha=1, Translation="6.624,-1.355,-2.057", Rotation="5.763,-57.942,-15.080") {
    rotationMode=BMM_Additive
    rotationSpace=BCS_BoneSpace
    translationMode=BMM_Additive
    translationSpace=BCS_BoneSpace
    alphaBoolBlend {blendOption=Linear}
    boneToModify {boneName=dyn_strap_fr_Top}
    rotation=<rot 5.763 -57.9419 -15.08>
    translation=<6.624 -1.355 -2.057>
   }
   n68 = "Transform (Modify) Bone - Bone: dyn_strap_fr_C"(ComponentPose=@n67, Alpha=1, Translation="-2.599,0.261,0.608", Rotation="-0.000,0.000,5.000") {
    rotationMode=BMM_Additive
    rotationSpace=BCS_BoneSpace
    translationMode=BMM_Additive
    translationSpace=BCS_BoneSpace
    alphaBoolBlend {blendOption=Linear}
    boneToModify {boneName=dyn_strap_fr_C}
    rotation=<rot 0 0.0001 5>
    translation=<-2.599 0.2608 0.608>
   }
   n69 = "Transform (Modify) Bone - Bone: dyn_strap_fr_Bot"(ComponentPose=@n68, Alpha=1, Translation="3.964,-0.301,0.000", Rotation="0,0,0") {
    rotationMode=BMM_Additive
    rotationSpace=BCS_BoneSpace
    translationMode=BMM_Additive
    translationSpace=BCS_BoneSpace
    alphaBoolBlend {blendOption=Linear}
    boneToModify {boneName=dyn_strap_fr_Bot}
    translation=<3.964 -0.301>
   }
   n70 = "Transform (Modify) Bone - Bone: dyn_strap_bk_Top"(ComponentPose=@n69, Alpha=1, Translation="-1.416,-0.149,0.804", Rotation="0,0,0") {
    rotationMode=BMM_Additive
    rotationSpace=BCS_BoneSpace
    translationMode=BMM_Additive
    translationSpace=BCS_BoneSpace
    alphaBoolBlend {blendOption=Linear}
    boneToModify {boneName=dyn_strap_bk_Top}
    translation=<-1.416 -0.149 0.804>
   }
   n71 = "Transform (Modify) Bone - Bone: dyn_strap_bk_C"(ComponentPose=@n70, Alpha=1, Translation="0.000,-0.000,-0.313", Rotation="0,0,0") {
    rotationMode=BMM_Additive
    rotationSpace=BCS_BoneSpace
    translationMode=BMM_Additive
    translationSpace=BCS_BoneSpace
    alphaBoolBlend {blendOption=Linear}
    boneToModify {boneName=dyn_strap_bk_C}
    translation {z=-0.313}
   }
   n72 = "Transform (Modify) Bone - Bone: dyn_strap_bk_Bot"(ComponentPose=@n71, Alpha=1, Translation="-0.531,0.188,-0.000", Rotation="0.013,-9.971,-19.988") {
    rotationMode=BMM_Additive
    rotationSpace=BCS_BoneSpace
    translationMode=BMM_Additive
    translationSpace=BCS_BoneSpace
    alphaBoolBlend {blendOption=Linear}
    boneToModify {boneName=dyn_strap_bk_Bot}
    rotation=<rot 0.013 -9.971 -19.988>
    translation=<-0.531 0.188>
   }
   n73 = "Transform (Modify) Bone - Bone: dyn_strap_fr_Top"(ComponentPose=@n72, Alpha=1, Translation="-0.046,1.126,-0.522", Rotation="-5.000,0.000,-35.000") {
    rotationMode=BMM_Additive
    rotationSpace=BCS_BoneSpace
    translationMode=BMM_Additive
    translationSpace=BCS_BoneSpace
    alphaBoolBlend {blendOption=Linear}
    boneToModify {boneName=dyn_strap_fr_Top}
    rotation=<rot -5 0 -35>
    translation=<-0.046 1.126 -0.522>
   }
   n74 = "Transform (Modify) Bone - Bone: dyn_strap_fr_C"(ComponentPose=@n73, Alpha=1, Translation="0,0,0", Rotation="0,0,0") {
    rotationMode=BMM_Additive
    rotationSpace=BCS_BoneSpace
    translationMode=BMM_Additive
    translationSpace=BCS_BoneSpace
    alphaBoolBlend {blendOption=Linear}
    boneToModify {boneName=dyn_strap_fr_C}
   }
   n75 = "Transform (Modify) Bone - Bone: dyn_strap_fr_Bot"(ComponentPose=@n74, Alpha=1, Translation="0,0,0", Rotation="0,0,0") {
    rotationMode=BMM_Additive
    rotationSpace=BCS_BoneSpace
    translationMode=BMM_Additive
    translationSpace=BCS_BoneSpace
    alphaBoolBlend {blendOption=Linear}
    boneToModify {boneName=dyn_strap_fr_Bot}
   }
   n76 = "Transform (Modify) Bone - Bone: dyn_strap_bk_Top"(ComponentPose=@n75, Alpha=1, Translation="-5.276,6.833,-2.363", Rotation="0.000,-32.008,-12.003") {
    rotationMode=BMM_Additive
    rotationSpace=BCS_BoneSpace
    translationMode=BMM_Additive
    translationSpace=BCS_BoneSpace
    alphaBoolBlend {blendOption=Linear}
    boneToModify {boneName=dyn_strap_bk_Top}
    rotation=<rot 0 -32.008 -12.003>
    translation=<-5.276 6.833 -2.363>
   }
   n77 = "Transform (Modify) Bone - Bone: dyn_strap_bk_C"(ComponentPose=@n76, Alpha=1, Translation="0,0,0", Rotation="0,0,0") {
    rotationMode=BMM_Additive
    rotationSpace=BCS_BoneSpace
    translationMode=BMM_Additive
    translationSpace=BCS_BoneSpace
    alphaBoolBlend {blendOption=Linear}
    boneToModify {boneName=dyn_strap_bk_C}
   }
   n78 = "Transform (Modify) Bone - Bone: dyn_strap_bk_Bot"(ComponentPose=@n77, Alpha=1, Translation="-1.529,-0.412,-0.817", Rotation="-20.000,-35.000,0.000") {
    rotationMode=BMM_Additive
    rotationSpace=BCS_BoneSpace
    translationMode=BMM_Additive
    translationSpace=BCS_BoneSpace
    alphaBoolBlend {blendOption=Linear}
    boneToModify {boneName=dyn_strap_bk_Bot}
    rotation=<rot -20 -35 0>
    translation=<-1.529 -0.412 -0.817>
   }
   n79 = Layered blend per bone(
    BasePose=Component To Local(ComponentPose=@n78)
    BlendPoses_0=Slot 'DefaultSlot'()
    BlendWeights_0=Get Custom Idle is Playing Float() -> (Custom Idle is Playing Float:float)
   ) {layerSetup [{branchFilters [{boneName=dyn_shoulder_MSTR}, {boneName=dyn_horn_l_01}, {boneName=dyn_horn_r_01}]}]}
   n80 = Get isDerbySwarmFE() -> (isDerbySwarmFE:bool)
   n81 = "Transform (Modify) Bone - Bone: dyn_shoulder_MSTR"!(
    ComponentPose=Local To Component(LocalPose=@n79)
    bAlphaBoolEnabled:bool=@n80
    Translation="0.2,-1.6,1.2"
    Rotation="-9.995,-30.925,-51.129"
    Scale="1,1,1"
   ) -> (Pose) {
    alphaInputType=Bool
    rotationMode=BMM_Additive
    translationMode=BMM_Additive
    alphaBoolBlend {blendOption=Linear}
    boneToModify {boneName=dyn_shoulder_MSTR}
    rotation=<rot -9.995 -30.925 -51.129>
    translation=<0.2 -1.6 1.2>
   }
   n82 = "Transform (Modify) Bone - Bone: dyn_horn_r_01"!(ComponentPose=@n81, bAlphaBoolEnabled:bool=@n80, Translation="0,0,0", Rotation="0,0,15", Scale="1,1,1") -> (Pose) {
    alphaInputType=Bool
    rotationMode=BMM_Additive
    translationMode=BMM_Additive
    alphaBoolBlend {blendOption=Linear}
    boneToModify {boneName=dyn_horn_r_01}
    rotation {roll=15}
   }
   n83 = "Transform (Modify) Bone - Bone: dyn_horn_r_02"!(ComponentPose=@n82, bAlphaBoolEnabled:bool=@n80, Translation="0,0,0", Rotation="0,5,5", Scale="1,1,1") -> (Pose) {
    alphaInputType=Bool
    rotationMode=BMM_Additive
    translationMode=BMM_Additive
    alphaBoolBlend {blendOption=Linear}
    boneToModify {boneName=dyn_horn_r_02}
    rotation {roll=5, yaw=5}
   }
   n84 = "Transform (Modify) Bone - Bone: dyn_horn_r_05"!(ComponentPose=@n83, bAlphaBoolEnabled:bool=@n80, Translation="0,0,0", Rotation="-3,0,-13", Scale="1,1,1") -> (Pose) {
    alphaInputType=Bool
    rotationMode=BMM_Additive
    translationMode=BMM_Additive
    alphaBoolBlend {blendOption=Linear}
    boneToModify {boneName=dyn_horn_r_05}
    rotation {pitch=-3, roll=-13}
   }
   n85 = "Constraint - dyn_strap_fr_C (PPPP)"(ComponentPose=@n84, Alpha=1, ConstraintWeights_0=0.4472, ConstraintWeights_1=0.0501, ConstraintWeights_2=0.0957, ConstraintWeights_3=0.407) {
    alphaBoolBlend {blendOption=Linear}
    alphaScaleBiasClamp {bClampResult=true, clampMax=0.3}
    boneToModify {boneName=dyn_strap_fr_C}
    constraintSetup [
     {targetBone {boneName=upperarm_twist_01_l}}
     {targetBone {boneName=spine_04}}
     {targetBone {boneName=spine_05}}
     {targetBone {boneName=clavicle_l}}
    ]
    constraintWeights [1,1,1,1]
   }
   n86 = "Constraint - dyn_strap_fr_Bot (PPPP)"(ComponentPose=@n85, Alpha=1, ConstraintWeights_0=0.1188, ConstraintWeights_1=0.2131, ConstraintWeights_2=0.2547, ConstraintWeights_3=0.4134) {
    alphaBoolBlend {blendOption=Linear}
    alphaScaleBiasClamp {bClampResult=true, clampMax=0.6}
    boneToModify {boneName=dyn_strap_fr_Bot}
    constraintSetup [
     {targetBone {boneName=upperarm_twist_01_l}}
     {targetBone {boneName=spine_04}}
     {targetBone {boneName=spine_05}}
     {targetBone {boneName=clavicle_l}}
    ]
    constraintWeights [1,1,1,1]
   }
   n87 = "Constraint - dyn_strap_bk_C (PPP)"(ComponentPose=@n86, Alpha=1, ConstraintWeights_0=0.2509, ConstraintWeights_1=0.031, ConstraintWeights_2=0.7181) {
    alphaBoolBlend {blendOption=Linear}
    alphaScaleBiasClamp {bClampResult=true, clampMax=0.3}
    boneToModify {boneName=dyn_strap_bk_C}
    constraintSetup [{targetBone {boneName=upperarm_twist_01_l}}, {targetBone {boneName=spine_05}}, {targetBone {boneName=clavicle_l}}]
    constraintWeights [1,1,1]
   }
   n88 = "Constraint - dyn_strap_bk_Bot (PPPPPPP)"(
    ComponentPose=@n87
    Alpha=1
    ConstraintWeights_0=0.0952
    ConstraintWeights_1=0
    ConstraintWeights_2=0.0347
    ConstraintWeights_3=0
    ConstraintWeights_4=0.1824
    ConstraintWeights_5=0.6863
    ConstraintWeights_6=0.0014
   ) {
    alphaBoolBlend {blendOption=Linear}
    alphaScaleBiasClamp {bClampResult=true, clampMax=0.6}
    boneToModify {boneName=dyn_strap_bk_Bot}
    constraintSetup [
     {targetBone {boneName=upperarm_twist_01_l}}
     {targetBone {boneName=spine_02}}
     {targetBone {boneName=spine_04}}
     {targetBone {boneName=spine_03}}
     {targetBone {boneName=spine_05}}
     {targetBone {boneName=clavicle_l}}
     {targetBone {boneName=clavicle_r}}
    ]
    constraintWeights [1,1,1,1,1,1,1]
   }
   Output Pose(Result=Component To Local(ComponentPose=@n88)) {layerGroup=None}
  ]}
 EventGraph {nodes [
   Event BlueprintUpdateAnimation()
   > n1 = Sequence!(execute) -> (then_0, then_1, then_2:Exec, then_3:Exec, then_4:Exec)
   > Set Skydive(Skydive:double=SelectFloat(A=1, B=0, bPickA=Get bIsSkyDiving())) -> (Output_Get:double)
   > Set ADS(ADS:double=SelectFloat(A=1, B=0, bPickA=Get bIsTargeting())) -> (Output_Get:double)
   > Set Emote(Emote:double=SelectFloat(A=1, B=0, bPickA=Get bIsPlayingEmote())) -> (Output_Get:double)
   > Set moving(moving:double=SelectFloat(A=1, B=0, bPickA=Get bIsMoving2D())) -> (Output_Get:double)
   > Set Crouch(Crouch:double=SelectFloat(A=1, B=0, bPickA=Get bIsCrouching())) -> (Output_Get:double)
   n2 = Get bIsJumping()
   n3 = Get bIsFalling()
   > Set Jump(Jump:double=SelectFloat(A=1, B=0, bPickA=OR Boolean(A=@n2, B=@n3))) -> (Output_Get:double)
   > Set parachute(parachute:double=SelectFloat(A=1, B=0, bPickA=Get bIsParachuteOpened())) -> (Output_Get:double)
   n4 = Gravity Override State Function DEPRECATED(
    execute=@n1.then_1
    CharacterPartAnimInstance=Self-Reference()
    GameTimeInSeconds=GetGameTimeInSeconds()
    GravityOverrideParams=Get Gravity Override Params()
   )
   > Set GravityOverride(GravityOverride=@n4)
   n5 = Cast To FortPlayerAnimInstance(execute=@n1.then_2, Object=GetAnimInstance(self=GetOwnerSkeletalMeshComponent()))
   > Set pistol(pistol:double=0) -> (Output_Get:double)
   > Set Melee(Melee:double=0) -> (Output_Get:double)
   > Set Rifle(Rifle:double=0) -> (Output_Get:double)
   > n6 = Switch on EFortWeaponCoreAnimation!(execute, Selection:EFortWeaponCoreAnimation Enum=Get WeaponCoreAnim(self=@n5)) -> (
    Melee:Exec
    Pistol:Exec
    Shotgun:Exec
    PaperBlueprint:Exec
    Rifle:Exec
    MeleeOneHand:Exec
    MachinePistol:Exec
    RocketLauncher:Exec
    GrenadeLauncher:Exec
    GoingCommando:Exec
    AssaultRifle:Exec
    TacticalShotgun:Exec
    SniperRifle:Exec
    TrapPlacement:Exec
    ShoulderLauncher:Exec
    AbilityDecoTool:Exec
    Crossbow:Exec
    C4:Exec
    RemoteControl:Exec
    DualWield:Exec
    AR_BullPup:Exec
    AR_ForwardGrip:Exec
    MedPackPaddles:Exec
    SMG_P90:Exec
    AR_DrumGun:Exec
    Consumable_Small:Exec
    Consumable_Large:Exec
    Balloon:Exec
    MountedTurret:Exec
    CreativeTool:Exec
    ExplosiveBow:Exec
    AshtonIndigo:Exec
    AshtonChicago:Exec
    MeleeDualWield:Exec
    Unarmed:Exec
   )
   Set Rifle(
    execute=[
     @n6.AR_BullPup
     @n6.AR_ForwardGrip
     @n6.AssaultRifle
     @n6.Crossbow
     @n6.GrenadeLauncher
     @n6.Rifle
     @n6.RocketLauncher
     @n6.Shotgun
     @n6.SniperRifle
     @n6.TacticalShotgun
    ]
    Rifle=1
   )
   Set Melee(execute=@n6.Melee, Melee=1)
   Set pistol(execute=@n6.Pistol, pistol=1)
   Branch(execute=@n1.then_3, Condition=Get bIsInFrontEnd())
   > Set Custom Idle is Playing Float(Custom Idle is Playing Float:float=GetCurveValue(self=GetOwnerFrontEndAnimInstance(), CurveName=Frontend_Idle_Bypass)) -> (Output_Get:float)
   n7 = Get bIsPlayingEmote()
   n8 = Branch(
    execute=@n1.then_4
    Condition=AND Boolean!(A=Get bIsInFrontEnd(), B=Get isDerbySwarmVariant() -> (isDerbySwarmVariant:bool), C:bool=NOT Boolean(A=@n7)) -> (ReturnValue)
   )
   > Set isDerbySwarmFE(isDerbySwarmFE:bool=true) -> (Output_Get:bool)
   Set isDerbySwarmFE(execute=@n8.else, isDerbySwarmFE=false)
   Event OnCharacterCustomizationLoaded()
   > n9 = Branch(Condition=CharacterItemDefinitionHasAnyMetaTag(Tags=Make Array!("[0]"=(TagName="Cosmetics.MetaTags.DerbySwarm"), "[1]":Gameplay Tag Structure=(TagName="")) -> (Array)))
   > Set isDerbySwarmVariant(isDerbySwarmVariant:bool=true) -> (Output_Get:bool)
   Set isDerbySwarmVariant(execute=@n9.else, isDerbySwarmVariant=false)
  ]}
]
variables [
 ADS {type=number}
 Crouch {type=number}
 Custom Idle is Playing Float {type=number}
 Emote {type=number}
 Gravity Override Params {type=GravityOverrideParamsStruct}
 GravityOverride {type=Vector}
 isDerbySwarmFE {type=boolean}
 isDerbySwarmVariant {type=boolean}
 Jump {type=number}
 Melee {type=number}
 moving {type=number}
 parachute {type=number}
 pistol {type=number}
 Rifle {type=number}
 Skydive {type=number}
]
