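# Post-process Control Rig for the PastaSauceMarinara female face accessory that applies corrective poses to its dynamic hair bones during skydiving, crouching and movement.
# A sequence runs three passes: a sky-gated skydive offset, a head-driven crouch-move correction, and a spine_02-driven crouch correction that constrains dyn_hair_03 toward upper spine and clavicles.
# Spherical pose readers and interpolated alphas weight the corrections, with Move and its inverse capping the two pose-driven alphas.
# Some constraints and a second head reader sit outside the execution chain.
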ControlRigBlueprint /BRCosmetics/Characters/Player/Female/Medium/Bodies/F_MED_PastaSauceMarinara/Meshes/Parts/F_MED_PastaSauceMarinara_FaceAcc_CtrlRig_Post
graphs [
 RigVMFunctionLibrary {links [], nodes []}
 RigVMFunction_Sequence_1_ContainedGraph {nodes [
   n1 = RigVMFunctionEntryNode()
   RigVMFunction_Sequence = &RigVMFunction_Sequence(ExecuteContext=@n1.ExecuteContext)
   n2 = &RigVMFunction_Sequence(ExecuteContext=@RigVMFunction_Sequence.B)
   n3 = &RigVMFunction_Sequence(ExecuteContext=@n2.B)
   RigVMFunctionReturnNode(A=@RigVMFunction_Sequence.A, B=@n2.A, C=@n3.A, D=@n3.B)
  ]}
 RigVMModel {nodes [
   n1 = &RigUnit_SphericalPoseReader(
    ActiveRegionScaleFactors=(PositiveWidth=1,NegativeWidth=1.5,PositiveHeight=1,NegativeHeight=1)
    ActiveRegionSize=0.2
    Debug=(bDrawDebug=false,bDraw2D=false,bDrawLocalAxes=false,DebugScale=25,DebugSegments=20,DebugThickness=0.25)
    DriverAxis=(X=1,Y=0,Z=0)
    DriverItem=(Type=Bone,Name="head")
    FalloffRegionScaleFactors=(PositiveWidth=0.3,NegativeWidth=1,PositiveHeight=0.5,NegativeHeight=0.5)
    FalloffSize=0.45
    OptionalParentItem=(Type=Bone,Name="neck_01")
    RotationOffset=(X=-90,Y=0,Z=50)
    ExecutePin=@RigVMFunction_Sequence.B
   )
   n2 = &RigVMFunction_AlphaInterp(
    InterpSpeedDecreasing=17
    InterpSpeedIncreasing=5
    bClampResult=true
    Value=@n1.OutputParam
    ClampMax=RigVMVariableNode() {variable=Move}.Value
    Scale=RigVMVariableNode() {variable=Crouch}.Value
   )
   &RigUnit_ParentConstraint(
    AdvancedSettings=(InterpolationType=Average,RotationOrderForFilter=XZY)
    Child=(Type=Bone,Name="dyn_hair_03")
    ExecutePin=()
    Filter=(TranslationFilter=(bX=false,bY=false,bZ=false),RotationFilter=(bX=true,bY=True,bZ=false),ScaleFilter=(bX=false,bY=false,bZ=false))
    Parents=((Item=(Type=Bone,Name="spine_05"),Weight=0.9959),(Item=(Type=Bone,Name="clavicle_l"),Weight=0.002),(Item=(Type=Bone,Name="clavicle_r"),Weight=0.002))
    Weight=@n2.Result
   )
   > &RigUnit_ParentConstraint(
    AdvancedSettings=(InterpolationType=Average,RotationOrderForFilter=XZY)
    Child=(Type=Bone,Name="dyn_hair_05")
    Filter=(TranslationFilter=(bX=false,bY=false,bZ=false),RotationFilter=(bX=true,bY=True,bZ=false),ScaleFilter=(bX=false,bY=false,bZ=false))
    Parents="((Item=(Type=Bone,Name=\"spine_02\"),Weight=0.0207),(Item=(Type=Bone,Name=\"spine_04\"),Weight=0.5089),(Item=(Type=Bone,Name=\"spine_03\"),Weight=0.0437),(Item=(Type=Bone,Name=\"spine_05\"),Weight=0.2692),(Item=(Type=Bone,Name=\"clavicle_l\"),Weight=0.0806),(Item=(T ...(truncated)"
    Weight=@n2.Result
   )
   RigUnit_BeginExecution = &RigUnit_BeginExecution()
   RigVMFunction_Sequence = RigVMAggregateNode(ExecuteContext=@RigUnit_BeginExecution.ExecutePin)
   RigVMFunction_ControlFlowBranch = &RigVMFunction_ControlFlowBranch(ExecuteContext=@RigVMFunction_Sequence.A, Condition=RigVMVariableNode() {variable=sky}.Value)
   AlphaInterp = &RigVMFunction_AlphaInterp(InterpSpeedDecreasing=15, InterpSpeedIncreasing=2, Scale=1, bClampResult=False, Value=RigVMVariableNode() {variable=Skydive}.Value)
   &RigUnit_ModifyTransforms(
    ItemToModify="((Item=(Type=Bone,Name=\"dyn_hair_01\"),Transform=(Rotation=(X=0,Y=0,Z=0,W=1),Translation=(X=5,Y=0,Z=3),Scale3D=(X=1,Y=1,Z=1))),(Item=(Type=Bone,Name=\"dyn_hair_l_MSTR\"),Transform=(Rotation=(X=0,Y=0,Z=0,W=1),Translation=(X=7,Y=0,Z=-2),Scale3D=(X=1,Y=1,Z=1))), ...(truncated)"
    WeightMinimum=0
    Weight=@AlphaInterp.Result
    ExecutePin=@RigVMFunction_ControlFlowBranch.True
    ItemToModify.0.Transform.Rotation=&RigVMFunction_MathQuaternionFromEuler(Euler=(X=0,Y=0,Z=-5)).Result
    ItemToModify.1.Transform.Rotation=&RigVMFunction_MathQuaternionFromEuler(Euler=(X=0,Y=-25,Z=10)).Result
    ItemToModify.2.Transform.Rotation=&RigVMFunction_MathQuaternionFromEuler(Euler=(X=0,Y=-25,Z=-15)).Result
    ItemToModify.3.Transform.Rotation=&RigVMFunction_MathQuaternionFromEuler(Euler=(X=0,Y=-20,Z=0)).Result
   )
   n3 = &RigUnit_SphericalPoseReader(
    ActiveRegionScaleFactors=(PositiveWidth=0.5,NegativeWidth=0.5,PositiveHeight=0.5,NegativeHeight=0.3)
    ActiveRegionSize=0.2
    Debug=(bDrawDebug=true,bDraw2D=false,bDrawLocalAxes=false,DebugScale=25,DebugSegments=20,DebugThickness=0.25)
    DriverAxis=(X=1,Y=0,Z=0)
    DriverItem=(Type=Bone,Name="spine_02")
    FalloffRegionScaleFactors=(PositiveWidth=0.2,NegativeWidth=0.1,PositiveHeight=0.2,NegativeHeight=0.1)
    FalloffSize=0.45
    FlipWidthScaling=false
    OptionalParentItem=(Type=Bone,Name="root")
    RotationOffset=(X=45,Y=-80,Z=0)
    ExecutePin=@RigVMFunction_Sequence.C
   )
   Remap = &RigVMFunction_MathDoubleRemap(SourceMaximum=1, SourceMinimum=0, TargetMaximum=0, TargetMinimum=1, bClamp=False, Value=RigVMVariableNode() {variable=Move}.Value)
   n4 = &RigVMFunction_AlphaInterp(
    InterpSpeedDecreasing=17
    InterpSpeedIncreasing=10
    bClampResult=true
    ClampMax=@Remap.Result
    Value=@n3.OutputParam
    Scale=RigVMVariableNode() {variable=Crouch}.Value
   )
   > &RigUnit_ParentConstraint(
    AdvancedSettings=(InterpolationType=Average,RotationOrderForFilter=XZY)
    Child=(Type=Bone,Name="dyn_hair_03")
    Filter=(TranslationFilter=(bX=false,bY=true,bZ=true),RotationFilter=(bX=false,bY=true,bZ=true),ScaleFilter=(bX=false,bY=false,bZ=false))
    Parents=((Item=(Type=Bone,Name="spine_04"),Weight=0.0439),(Item=(Type=Bone,Name="spine_05"),Weight=0.6679),(Item=(Type=Bone,Name="clavicle_l"),Weight=0.1786),(Item=(Type=Bone,Name="clavicle_r"),Weight=0.1096))
    Weight=@n4.Result
   )
   RigVMFunction_MathQuaternionFromEuler = &RigVMFunction_MathQuaternionFromEuler(Euler=(X=0,Y=-3,Z=0))
   > &RigUnit_ModifyTransforms(
    ItemToModify="((Item=(Type=Bone,Name=\"dyn_hair_01\"),Transform=(Rotation=(X=0,Y=0,Z=0,W=1),Translation=(X=0,Y=0,Z=2),Scale3D=(X=1,Y=1,Z=1))),(Item=(Type=Bone,Name=\"dyn_hair_02\"),Transform=(Rotation=(X=0,Y=0,Z=0,W=1),Translation=(X=-1.3,Y=-2,Z=0),Scale3D=(X=1,Y=1,Z=1))),( ...(truncated)"
    WeightMinimum=0
    Weight=@n4.Result
    ItemToModify.2.Transform.Rotation=@RigVMFunction_MathQuaternionFromEuler.Result
    ItemToModify.3.Transform.Rotation=&RigVMFunction_MathQuaternionFromEuler(Euler=(X=0,Y=0,Z=-15)).Result
   )
   n5 = &RigUnit_SphericalPoseReader(
    ActiveRegionScaleFactors=(PositiveWidth=0.7,NegativeWidth=0.2,PositiveHeight=0.7,NegativeHeight=0.5)
    ActiveRegionSize=0.2
    Debug=(bDrawDebug=true,DebugScale=25,DebugSegments=20,DebugThickness=0.25)
    DriverAxis=(X=1,Y=0,Z=0)
    DriverItem=(Type=Bone,Name="head")
    ExecutePin=()
    FalloffRegionScaleFactors=(PositiveWidth=0.4,NegativeWidth=0.1,PositiveHeight=0.4,NegativeHeight=0.2)
    FalloffSize=0.45
    OptionalParentItem=(Type=Bone,Name="neck_01")
    RotationOffset=(X=-90,Y=0,Z=110)
   )
   > &RigUnit_ModifyTransforms(
    ItemToModify="((Item=(Type=Bone,Name=\"dyn_hair_03\"),Transform=(Rotation=(X=0,Y=0,Z=0,W=1),Translation=(X=0,Y=0,Z=-2),Scale3D=(X=1,Y=1,Z=1))),(Item=(Type=Bone,Name=\"dyn_hair_l_MSTR\"),Transform=(Rotation=(X=0,Y=0,Z=0,W=1),Translation=(X=0,Y=0,Z=0),Scale3D=(X=1,Y=1,Z=1))), ...(truncated)"
    WeightMinimum=0
    ItemToModify.0.Transform.Rotation=&RigVMFunction_MathQuaternionFromEuler(Euler=(X=0,Y=2,Z=0), RotationOrder=XYZ).Result
    Weight=@n5.OutputParam
   )
   n6 = &RigVMFunction_AlphaInterp(
    InterpSpeedDecreasing=7
    InterpSpeedIncreasing=2
    bClampResult=false
    Scale=RigVMVariableNode() {variable=Move}.Value
    Value=RigVMVariableNode() {variable=Crouch}.Value
   )
   &RigUnit_ModifyTransforms(
    ItemToModify="((Item=(Type=Bone,Name=\"dyn_hair_05\"),Transform=(Rotation=(X=0,Y=0,Z=0,W=1),Translation=(X=0,Y=0,Z=0),Scale3D=(X=1,Y=1,Z=1))),(Item=(Type=Bone,Name=\"dyn_hair_01\"),Transform=(Rotation=(X=0,Y=0,Z=0,W=1),Translation=(X=0,Y=0,Z=0),Scale3D=(X=1,Y=1,Z=1))),(Item ...(truncated)"
    WeightMinimum=0
    Weight=@n6.Result
    ExecutePin=@n7.ExecutePin
    ItemToModify.0.Transform.Rotation=&RigVMFunction_MathQuaternionFromEuler(Euler=(X=0,Y=-10,Z=0)).Result
   )
   n7 = &RigUnit_ModifyTransforms(
    ItemToModify="((Item=(Type=Bone,Name=\"dyn_hair_07\"),Transform=(Rotation=(X=0,Y=0,Z=0,W=1),Translation=(X=0,Y=0,Z=0),Scale3D=(X=1,Y=1,Z=1))),(Item=(Type=Bone,Name=\"dyn_hair_05\"),Transform=(Rotation=(X=0,Y=0,Z=0,W=1),Translation=(X=0,Y=0,Z=0),Scale3D=(X=1,Y=1,Z=1))),(Item ...(truncated)"
    WeightMinimum=0
    Weight=@n2.Result
    ItemToModify.0.Transform.Rotation=&RigVMFunction_MathQuaternionFromEuler(Euler=(X=0,Y=5,Z=0)).Result
    ItemToModify.1.Transform.Rotation=&RigVMFunction_MathQuaternionFromEuler(Euler=(X=0,Y=5,Z=0)).Result
    ItemToModify.2.Transform.Rotation=&RigVMFunction_MathQuaternionFromEuler(Euler=(X=0,Y=-2,Z=0)).Result
    ItemToModify.3.Transform.Rotation=&RigVMFunction_MathQuaternionFromEuler(Euler=(X=0,Y=-20,Z=10)).Result
    ExecutePin=@n1.ExecutePin
   )
  ]}
]
hierarchy [
 root {loc=<0 0 0>, rot=<0 0 0 1>}
 pelvis {parent=root, loc=<0 -0.5301 98.0571>, rot=<0.0127 0.707 -0.0127 -0.707>}
 spine_01 {parent=pelvis, loc=<6.1208 -0.9328 0>, rot=<0 0 0.034 -0.9994>}
 spine_02 {parent=spine_01, loc=<6.5996 -0.2095 0>, rot=<0 0 0.0238 -0.9997>}
 spine_03 {parent=spine_02, loc=<6.4914 -0.1237 0>, rot=<0 0 0.0149 -0.9999>}
 spine_04 {parent=spine_03, loc=<6.9598 0.0794 0>, rot=<0 0 -0.0132 -0.9999>}
 spine_05 {parent=spine_04, loc=<6.9987 0 0>, rot=<0 0 -0.0015 -1>}
 neck_01 {parent=spine_05, loc=<11.3138 0.1529 0>, rot=<0 0 -0.2003 -0.9797>}
 neck_02 {parent=neck_01, loc=<4.5627 0 0>, rot=<0 0 0.0504 -0.9987>}
 head {parent=neck_02, loc=<5.3914 0 0>, rot=<0 0 0.115 -0.9934>}
 dyn_hair_01 {parent=head, loc=<6.197 -5.2172 0.1989>, rot=<0.2063 -0.6736 0.6763 -0.2151>}
 dyn_hair_02 {parent=dyn_hair_01, loc=<7.9477 -0.0003 -0.3981>, rot=<0.0002 -0.1215 0.0004 -0.9926>}
 dyn_hair_03 {parent=dyn_hair_02, loc=<10.1601 -0.0139 -1.1693>, rot=<0.0004 -0.1348 0.0018 -0.9909>}
 dyn_hair_04 {parent=dyn_hair_03, loc=<7.9587 -0.0114 -0.2532>, rot=<0.0001 -0.047 0.0034 -0.9989>}
 dyn_hair_05 {parent=dyn_hair_04, loc=<5.5127 0 0>, rot=<-0.0001 -0.0063 0.0038 -1>}
 dyn_hair_06 {parent=dyn_hair_05, loc=<6.1149 0 0>, rot=<-0.0001 0.0047 0.0061 -1>}
 dyn_hair_07 {parent=dyn_hair_06, loc=<6.4252 0 0>, rot=<0 0.0042 0.0027 -1>}
 dyn_hair_08 {parent=dyn_hair_07, loc=<5.489 0 0>, rot=<0 0.0007 -0.0007 -1>}
 dyn_hair_r_MSTR {parent=head, loc=<1.2748 -3.5818 8.2906>, rot=<0.1766 -0.6801 0.6583 -0.2699>}
 dyn_hair_r_01 {parent=dyn_hair_r_MSTR, loc=<0 0 0>, rot=<0 0 0 1>}
 dyn_hair_r_02 {parent=dyn_hair_r_01, loc=<8.9599 -0.0012 -0.6497>, rot=<0.0002 -0.1215 0.0004 -0.9926>}
 dyn_hair_l_MSTR {parent=head, loc=<1.2626 -3.5701 -7.3142>, rot=<0.247 -0.6608 0.68 -0.1998>}
 dyn_hair_l_01 {parent=dyn_hair_l_MSTR, loc=<0 0 0>, rot=<0 0 0 1>}
 dyn_hair_l_02 {parent=dyn_hair_l_01, loc=<8.9599 -0.0012 -0.6497>, rot=<0.0002 -0.1215 0.0004 -0.9926>}
 clavicle_l {parent=spine_05, loc=<5.4831 -0.1053 -3.2946>, rot=<0.0465 -0.7289 -0.009 -0.6829>}
 upperarm_l {parent=clavicle_l, loc=<9.8658 -1.2136 -0.4105>, rot=<0.0164 -0.3245 -0.0269 -0.9454>}
 upperarm_twist_01_l {parent=upperarm_l, loc=<7.2135 -0.1134 0>, rot=<0 0 0 1>}
 dyn_horn_l_01 {parent=clavicle_l, loc=<6.091 -10.156 -1.2203>, rot=<-0.0854 0.6796 0.0545 -0.7266>}
 dyn_horn_l_02 {parent=dyn_horn_l_01, loc=<2.6659 0 0>, rot=<0.0032 -0.0853 -0.1647 -0.9826>}
 dyn_horn_l_03 {parent=dyn_horn_l_02, loc=<3.4598 0 0>, rot=<0.0069 0.022 -0.1757 -0.9842>}
 dyn_horn_l_04 {parent=dyn_horn_l_03, loc=<4.0416 0 0>, rot=<-0.008 -0.009 -0.2376 -0.9713>}
 dyn_horn_l_05 {parent=dyn_horn_l_04, loc=<3.6107 0 0>, rot=<-0.0351 0.0324 -0.2519 -0.9666>}
 dyn_horn_l_06 {parent=dyn_horn_l_05, loc=<4.3845 -0.0001 -0.1622>, rot=<-0.0532 0.0012 -0.2599 -0.9642>}
 dyn_horn_l_07 {parent=dyn_horn_l_06, loc=<5.365 -0.0193 -0.3725>, rot=<0.0236 -0.0117 -0.2693 -0.9627>}
 dyn_horn_l_08 {parent=dyn_horn_l_07, loc=<6.4273 0.1226 0.0236>, rot=<-0.1701 -0.0517 -0.0533 -0.9826>}
 dyn_horn_l_09 {parent=dyn_horn_l_08, loc=<4.9697 -0.0373 0.0049>, rot=<0.0573 -0.0572 0.1186 -0.9896>}
 dyn_horn_l_010 {parent=dyn_horn_l_09, loc=<5.2721 0 0.373>, rot=<0.121 -0.1899 -0.0126 -0.9742>}
 dyn_horn_l_011 {parent=dyn_horn_l_010, loc=<4.1528 0 0>, rot=<0.0694 -0.2291 -0.1067 -0.965>}
 dyn_horn_l_012 {parent=dyn_horn_l_011, loc=<3.1043 0 0>, rot=<-0.0078 -0.1905 -0.1904 -0.963>}
 dyn_horn_l_013 {parent=dyn_horn_l_012, loc=<4.4845 0 0>, rot=<-0.0095 -0.0343 -0.0171 -0.9992>}
 clavicle_r {parent=spine_05, loc=<5.4831 -0.1053 3.2946>, rot=<0.7289 0.0465 -0.6829 0.009>}
 upperarm_r {parent=clavicle_r, loc=<-9.8658 1.2136 0.4105>, rot=<0.0164 -0.3245 -0.0269 -0.9454>}
 upperarm_twist_01_r {parent=upperarm_r, loc=<-7.2135 0.1134 0>, rot=<0 0 0 -1>}
 dyn_horn_r_01 {parent=clavicle_r, loc=<-6.091 10.1561 1.2201>, rot=<-0.0854 0.6796 0.0545 -0.7266>}
 dyn_horn_r_02 {parent=dyn_horn_r_01, loc=<-2.6655 0.0001 0.0001>, rot=<0.0032 -0.0853 -0.1647 -0.9826>}
 dyn_horn_r_03 {parent=dyn_horn_r_02, loc=<-3.4598 -0.0001 0>, rot=<0.0069 0.022 -0.1757 -0.9842>}
 dyn_horn_r_04 {parent=dyn_horn_r_03, loc=<-4.0416 0 0>, rot=<-0.008 -0.009 -0.2376 -0.9713>}
 dyn_horn_r_05 {parent=dyn_horn_r_04, loc=<-3.6111 0.0006 -0.0003>, rot=<-0.0351 0.0324 -0.2519 -0.9666>}
 dyn_horn_r_06 {parent=dyn_horn_r_05, loc=<-4.3845 -0.0006 0.1626>, rot=<-0.0532 0.0012 -0.2599 -0.9642>}
 dyn_horn_r_07 {parent=dyn_horn_r_06, loc=<-5.365 0.0194 0.3724>, rot=<0.0236 -0.0117 -0.2693 -0.9627>}
 dyn_horn_r_08 {parent=dyn_horn_r_07, loc=<-6.4274 -0.1227 -0.0235>, rot=<-0.1701 -0.0517 -0.0533 -0.9826>}
 dyn_horn_r_09 {parent=dyn_horn_r_08, loc=<-4.9689 0.0375 -0.0053>, rot=<0.0573 -0.0572 0.1186 -0.9896>}
 dyn_horn_r_010 {parent=dyn_horn_r_09, loc=<-5.2725 -0.0002 -0.3729>, rot=<0.121 -0.1899 -0.0126 -0.9742>}
 dyn_horn_r_011 {parent=dyn_horn_r_010, loc=<-4.1527 0.0001 0>, rot=<0.0694 -0.2291 -0.1067 -0.965>}
 dyn_horn_r_012 {parent=dyn_horn_r_011, loc=<-3.1046 -0.0001 -0.0003>, rot=<-0.0078 -0.1905 -0.1904 -0.963>}
 dyn_horn_r_013 {parent=dyn_horn_r_012, loc=<-4.4845 0 0.0001>, rot=<-0.0095 -0.0343 -0.0171 -0.9992>}
 thigh_l {parent=pelvis, loc=<-1.0855 -0.2109 -8.7053>, rot=<0.0009 0.0355 0.0155 -0.9993>}
 thigh_twist_01_l {parent=thigh_l, loc=<-27.952 0 0>, rot=<0.0013 0 0 -1>}
 root {parent=root, type=Socket, loc=<0 0 0>, rot=<0 0 0 1>}
]
variables [
 Crouch {type=double, default=0}
 Move {type=double, default=0}
 Skydive {type=double, default=0}
 sky {type=bool, default=False}
]
