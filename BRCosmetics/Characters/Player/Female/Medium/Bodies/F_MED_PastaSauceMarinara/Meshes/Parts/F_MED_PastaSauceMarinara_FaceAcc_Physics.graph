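# Physics asset for the face accessory part of the PastaSauceMarinara female medium cosmetic body, covering dynamic hair and horn bones.
# Capsules and spheres cover pelvis, spine, neck and head as well as the dynamic bones.
# An eight-link hair chain and two short side chains hang from the head with tight angular limits, loosening along the main chain.
# Mirrored horn chains attach to spine_05 with free rotation, as do body joints and side-chain roots.
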PhysicsAsset /BRCosmetics/Characters/Player/Female/Medium/Bodies/F_MED_PastaSauceMarinara/Meshes/Parts/F_MED_PastaSauceMarinara_FaceAcc_Physics
bodies {
 dyn_hair_01 [{shapeType=Capsule, length=8.4688, radius=4, shapeName=dyn_hair_01_capsule, center=<3 0 0>, rotation=<rot 90 0 0>}]
 dyn_hair_02 [{shapeType=Capsule, length=8.5789, radius=4.8, shapeName=dyn_hair_02_capsule, center=<3 0 0>, rotation=<rot 90 0 0>}]
 dyn_hair_03 [{shapeType=Capsule, length=12.0609, radius=4.9887, shapeName=dyn_hair_03_capsule, center=<3 0 3>, rotation=<rot 90 0 0>}]
 dyn_hair_04 [{shapeType=Capsule, length=13.7041, radius=5.0178, shapeName=dyn_hair_04_capsule, center=<3 0 3.5>, rotation=<rot 90 0 0>}]
 dyn_hair_05 [{shapeType=Capsule, length=6.5, radius=5.75, shapeName=dyn_hair_05_capsule, center=<3 0 2>, rotation=<rot 90 0 0>}]
 dyn_hair_06 [{shapeType=Capsule, length=9.0938, radius=4.7188, shapeName=dyn_hair_06_capsule, center=<3 0 0>, rotation=<rot 90 0 0>}]
 dyn_hair_07 [{shapeType=Capsule, length=9.5938, radius=4.2188, shapeName=dyn_hair_07_capsule, center=<3 0 0>, rotation=<rot 90 0 0>}]
 dyn_hair_08 [{shapeType=Capsule, length=11.0938, radius=2.7188, shapeName=dyn_hair_08_capsule, center=<3 0 0>, rotation=<rot 90 0 0>}]
 dyn_hair_l_01 [{shapeType=Capsule, length=5.9375, radius=2.4688, shapeName=dyn_hair_l_01_capsule, rotation=<rot 90 0 0>}]
 dyn_hair_l_02 [
  {shapeType=Sphere, radius=1.6429, shapeName=dyn_hair_l_02_sphere, center=<9.2365 -1.5449 0>}
  {shapeType=Capsule, length=6.125, radius=3, shapeName=dyn_hair_l_02_capsule, center=<0 -4 0>, rotation=<rot 90 0 0>}
 ]
 dyn_hair_l_MSTR [{shapeType=Sphere, radius=0.505, shapeName=dyn_hair_l_MSTR_sphere}]
 dyn_hair_r_01 [{shapeType=Capsule, length=5.9375, radius=2.4688, shapeName=dyn_hair_r_01_capsule, rotation=<rot 90 0 0>}]
 dyn_hair_r_02 [
  {shapeType=Sphere, radius=1.6429, shapeName=dyn_hair_r_02_sphere, center=<9.2365 1.7518 0>}
  {shapeType=Capsule, length=6.125, radius=3, shapeName=dyn_hair_r_02_capsule, center=<0 4 0>, rotation=<rot 90 0 0>}
 ]
 dyn_hair_r_MSTR [{shapeType=Sphere, radius=0.505, shapeName=dyn_hair_r_MSTR_sphere}]
 dyn_horn_l_02 [{shapeType=Capsule, length=23.3563, radius=3.9299, shapeName=dyn_horn_l_02_capsule, center=<-1.686 0 -0.8992>, rotation=<rot 0 0 90>}]
 dyn_horn_l_03 [{
   shapeType=Capsule
   length=17.2091
   radius=3.4819
   shapeName=dyn_horn_l_03_capsule
   center=<0 0 -1.0776>
   rotation=<rot 80 -179.9999 -179.9997>
  }]
 dyn_horn_l_05 [{shapeType=Capsule, length=29.6683, radius=4.2405, shapeName=dyn_horn_l_05_capsule, rotation=<rot 90 0 0>}]
 dyn_horn_r_02 [{shapeType=Capsule, length=23.3563, radius=3.9299, shapeName=dyn_horn_r_02_capsule, center=<1.9709 0 1.1129>, rotation=<rot 0 0 90>}]
 dyn_horn_r_03 [{
   shapeType=Capsule
   length=17.2091
   radius=3.4819
   shapeName=dyn_horn_r_03_capsule
   center=<0 0 0.9532>
   rotation=<rot 80 -179.9999 -179.9997>
  }]
 dyn_horn_r_05 [{shapeType=Capsule, length=29.6683, radius=4.2405, shapeName=dyn_horn_r_05_capsule, rotation=<rot 90 0 0>}]
 head [{shapeType=Sphere, radius=8.5, shapeName=head_sphere, center=<0 -0.3204 0>}]
 neck_01 [{shapeType=Capsule, length=21.7504, radius=5.2496, shapeName=neck_01_capsule}]
 neck_02 [{shapeType=Capsule, length=15.7505, radius=7.2496, shapeName=neck_02_capsule}]
 pelvis [{shapeType=Capsule, length=17.1924, radius=10.3298, shapeName=pelvis_capsule, center=<-3.3513 -1.283 0>, rotation=<rot 0 -90 0>}]
 spine_01 [{shapeType=Capsule, length=22.195, radius=6.635, shapeName=spine_01_capsule, rotation=<rot 0 -90 0>}]
 spine_02 [{shapeType=Capsule, length=20.8603, radius=9.1875, shapeName=spine_02_capsule, center=<-2.8793 1.2982 0>, rotation=<rot 0 -90 0>}]
 spine_03 [{shapeType=Capsule, length=22.195, radius=6.635, shapeName=spine_03_capsule, rotation=<rot 0 -90 0>}]
 spine_04 [
  {shapeType=Capsule, length=22.3125, radius=11.0938, shapeName=spine_04_capsule, center=<1.6059 1.3994 0>, rotation=<rot 0 4.9992 0>}
  {shapeType=Capsule, length=24.1791, radius=9.4688, shapeName=spine_04_capsule, center=<-4.5797 2.2128 0>}
 ]
 spine_05 [{shapeType=Capsule, length=35, radius=9.5, shapeName=spine_05_capsule, center=<6.7508 1.1083 0>, rotation=<rot 0 -175.0005 180>}]
}
constraints [
 {
  bone1Name=spine_05
  bone2Name=spine_04
  swing1LimitDegrees=45
  swing1Motion=Free
  swing2LimitDegrees=45
  swing2Motion=Free
  twistLimitDegrees=45
  twistMotion=Free
 }
 {
  bone1Name=dyn_horn_l_02
  bone2Name=spine_05
  swing1LimitDegrees=45
  swing1Motion=Free
  swing2LimitDegrees=45
  swing2Motion=Free
  twistLimitDegrees=45
  twistMotion=Free
 }
 {
  bone1Name=dyn_horn_l_03
  bone2Name=dyn_horn_l_02
  swing1LimitDegrees=45
  swing1Motion=Free
  swing2LimitDegrees=45
  swing2Motion=Free
  twistLimitDegrees=45
  twistMotion=Free
 }
 {
  bone1Name=dyn_horn_r_02
  bone2Name=spine_05
  swing1LimitDegrees=45
  swing1Motion=Free
  swing2LimitDegrees=45
  swing2Motion=Free
  twistLimitDegrees=45
  twistMotion=Free
 }
 {
  bone1Name=dyn_horn_r_03
  bone2Name=dyn_horn_r_02
  swing1LimitDegrees=45
  swing1Motion=Free
  swing2LimitDegrees=45
  swing2Motion=Free
  twistLimitDegrees=45
  twistMotion=Free
 }
 {bone1Name=dyn_hair_01, bone2Name=head, swing1LimitDegrees=10, swing2LimitDegrees=45, twistLimitDegrees=2}
 {bone1Name=dyn_hair_02, bone2Name=dyn_hair_01, swing1LimitDegrees=10, swing2LimitDegrees=45, twistLimitDegrees=2}
 {bone1Name=dyn_hair_03, bone2Name=dyn_hair_02, swing1LimitDegrees=20, swing2LimitDegrees=60, twistLimitDegrees=4}
 {bone1Name=dyn_hair_04, bone2Name=dyn_hair_03, swing1LimitDegrees=35, swing2LimitDegrees=60, twistLimitDegrees=5}
 {bone1Name=dyn_hair_05, bone2Name=dyn_hair_04, swing1LimitDegrees=35, swing2LimitDegrees=50, twistLimitDegrees=5}
 {bone1Name=dyn_hair_06, bone2Name=dyn_hair_05, swing1LimitDegrees=35, swing2LimitDegrees=50, twistLimitDegrees=5}
 {bone1Name=dyn_hair_07, bone2Name=dyn_hair_06, swing1LimitDegrees=35, swing2LimitDegrees=50, twistLimitDegrees=5}
 {bone1Name=dyn_hair_08, bone2Name=dyn_hair_07, swing1LimitDegrees=35, swing2LimitDegrees=50, twistLimitDegrees=5}
 {
  bone1Name=neck_01
  bone2Name=spine_05
  swing1LimitDegrees=45
  swing1Motion=Free
  swing2LimitDegrees=45
  swing2Motion=Free
  twistLimitDegrees=45
  twistMotion=Free
 }
 {
  bone1Name=neck_02
  bone2Name=neck_01
  swing1LimitDegrees=45
  swing1Motion=Free
  swing2LimitDegrees=45
  swing2Motion=Free
  twistLimitDegrees=45
  twistMotion=Free
 }
 {
  bone1Name=head
  bone2Name=neck_02
  swing1LimitDegrees=45
  swing1Motion=Free
  swing2LimitDegrees=45
  swing2Motion=Free
  twistLimitDegrees=45
  twistMotion=Free
 }
 {
  bone1Name=spine_04
  bone2Name=spine_03
  swing1LimitDegrees=45
  swing1Motion=Free
  swing2LimitDegrees=45
  swing2Motion=Free
  twistLimitDegrees=45
  twistMotion=Free
 }
 {
  bone1Name=spine_02
  bone2Name=spine_01
  swing1LimitDegrees=45
  swing1Motion=Free
  swing2LimitDegrees=45
  swing2Motion=Free
  twistLimitDegrees=45
  twistMotion=Free
 }
 {bone1Name=dyn_hair_l_01, bone2Name=dyn_hair_l_MSTR, swing1LimitDegrees=20, swing2LimitDegrees=20, twistLimitDegrees=5}
 {bone1Name=dyn_hair_l_02, bone2Name=dyn_hair_l_01, swing1LimitDegrees=15, swing2LimitDegrees=20, twistLimitDegrees=5}
 {bone1Name=dyn_hair_r_01, bone2Name=dyn_hair_r_MSTR, swing1LimitDegrees=20, swing2LimitDegrees=20, twistLimitDegrees=5}
 {bone1Name=dyn_hair_r_02, bone2Name=dyn_hair_r_01, swing1LimitDegrees=15, swing2LimitDegrees=20, twistLimitDegrees=5}
 {
  bone1Name=dyn_hair_l_MSTR
  bone2Name=head
  swing1LimitDegrees=45
  swing1Motion=Free
  swing2LimitDegrees=45
  swing2Motion=Free
  twistLimitDegrees=45
  twistMotion=Free
 }
 {
  bone1Name=dyn_hair_r_MSTR
  bone2Name=head
  swing1LimitDegrees=45
  swing1Motion=Free
  swing2LimitDegrees=45
  swing2Motion=Free
  twistLimitDegrees=45
  twistMotion=Free
 }
 {
  bone1Name=dyn_horn_l_05
  bone2Name=dyn_horn_l_03
  swing1LimitDegrees=45
  swing1Motion=Free
  swing2LimitDegrees=45
  swing2Motion=Free
  twistLimitDegrees=45
  twistMotion=Free
 }
 {
  bone1Name=dyn_horn_r_05
  bone2Name=dyn_horn_r_03
  swing1LimitDegrees=45
  swing1Motion=Free
  swing2LimitDegrees=45
  swing2Motion=Free
  twistLimitDegrees=45
  twistMotion=Free
 }
 {
  bone1Name=spine_03
  bone2Name=spine_02
  swing1LimitDegrees=45
  swing1Motion=Free
  swing2LimitDegrees=45
  swing2Motion=Free
  twistLimitDegrees=45
  twistMotion=Free
 }
 {
  bone1Name=spine_01
  bone2Name=pelvis
  swing1LimitDegrees=45
  swing1Motion=Free
  swing2LimitDegrees=45
  swing2Motion=Free
  twistLimitDegrees=45
  twistMotion=Free
 }
]
