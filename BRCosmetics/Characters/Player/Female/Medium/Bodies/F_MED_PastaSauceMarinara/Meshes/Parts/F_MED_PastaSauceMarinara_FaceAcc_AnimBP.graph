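# Anim Blueprint for the female PastaSauceMarinara face accessory mesh, adding secondary hair motion over poses copied from parent and body meshes.
# A state-driven Control Rig feeds three RigidBody setups, blended by crouch and skydive.
# A post Control Rig follows, then per-bone AnimDynamics with spherical limits, plus bone offsets and constraints shaping hair around neck, shoulders and spine.
# The event graph updates gravity override, back bling and movement-state values.
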AnimBlueprint /BRCosmetics/Characters/Player/Female/Medium/Bodies/F_MED_PastaSauceMarinara/Meshes/Parts/F_MED_PastaSauceMarinara_FaceAcc_AnimBP
target_skeleton=&/F_MED_PastaSauceMarinara_FaceAcc_Skeleton
graphs [
 AnimGraph {nodes [
   n1 = Control Rig!(
    Source=Use cached pose 'RBAN'()
    Alpha
    Move:double=Get moving() -> (moving:double)
    Crouch:double=Get Crouch() -> (Crouch:double)
    Skydive:double=Get Skydive() -> (Skydive:double)
    sky:bool=Get bIsSkyDiving()
   ) -> (Pose) {
    alphaBoolBlend {blendOption=Linear}
    controlRigAssetReference {blueprintRigClass=&/F_MED_PastaSauceMarinara_FaceAcc_CtrlRig_Post.F_MED_PastaSauceMarinara_FaceAcc_CtrlRig_Post_C}
   }
   n2 = Anim Dynamics - Bone: dyn_hair_l_01!(ComponentPose=Local To Component(LocalPose=@n1), Alpha=1, GravityOverride:Vector="0,0,-980", ExternalForce="0,0,0") -> (Pose) {
    angularDampingOverride=0.9999
    angularSpringConstant=500
    bAngularSpring=true
    bGravityOverrideInSimSpace=true
    bLinearSpring=true
    bOverrideAngularDamping=true
    bUseGravityOverride=true
    bUseSphericalLimits=true
    linearSpringConstant=50
    simulationSpace=BoneRelative
    alphaBoolBlend {blendOption=Linear}
    boundBone {boneName=dyn_hair_l_01}
    componentAppliedLinearAccClamp=<1500 1500 0>
    componentLinearAccScale=<0.3 0.3 0.3>
    componentLinearVelScale=<0.3 0.3 0.3>
    gravityOverride {z=-980}
    physicsBodyDefinitions [{
      collisionType=CustomSphere
      sphereCollisionRadius=1
      boundBone {boneName=dyn_hair_l_01}
      boxExtents=<5 20 20>
      constraintSetup {angularLimitsMax=<1 2 10>, angularLimitsMin=<-1 -20 -2>, angularTarget=<0 0 0>}
      localJointOffset=<3 0 0>
     }]
    relativeSpaceBone {boneName=root}
    sphericalLimits [{limitRadius=5, drivingBone {boneName=head}, sphereLocalOffset=<0 3 0>}]
   }
   n3 = Anim Dynamics - Bone: dyn_hair_l_02!(ComponentPose=@n2, Alpha=1, GravityOverride:Vector="0,0,-980", ExternalForce="0,0,0") -> (Pose) {
    angularDampingOverride=0.9999
    angularSpringConstant=500
    bAngularSpring=true
    bGravityOverrideInSimSpace=true
    bLinearSpring=true
    bOverrideAngularDamping=true
    bUseGravityOverride=true
    bUseSphericalLimits=true
    linearSpringConstant=50
    simulationSpace=BoneRelative
    alphaBoolBlend {blendOption=Linear}
    boundBone {boneName=dyn_hair_l_02}
    componentAppliedLinearAccClamp=<1500 1500 0>
    componentLinearAccScale=<0.3 0.3 0.3>
    componentLinearVelScale=<0.3 0.3 0.3>
    gravityOverride {z=-980}
    physicsBodyDefinitions [{
      collisionType=CustomSphere
      sphereCollisionRadius=1
      boundBone {boneName=dyn_hair_l_02}
      boxExtents=<5 20 20>
      constraintSetup {angularLimitsMax=<1 2 10>, angularLimitsMin=<-1 -20 -2>, angularTarget=<0 0 0>}
      localJointOffset=<3 0 0>
     }]
    relativeSpaceBone {boneName=root}
    sphericalLimits [{limitRadius=5, drivingBone {boneName=head}, sphereLocalOffset=<0 3 0>}]
   }
   n4 = Anim Dynamics - Bone: dyn_hair_01!(ComponentPose=@n3, Alpha=1, GravityOverride:Vector="0,0,-980", ExternalForce) -> (Pose) {
    angularDampingOverride=0.9999
    angularSpringConstant=500
    bAngularSpring=true
    bGravityOverrideInSimSpace=true
    bLinearSpring=true
    bOverrideAngularDamping=true
    bUseGravityOverride=true
    bUseSphericalLimits=true
    linearSpringConstant=50
    simulationSpace=BoneRelative
    alphaBoolBlend {blendOption=Linear}
    boundBone {boneName=dyn_hair_01}
    componentAppliedLinearAccClamp=<1500 1500 0>
    componentLinearAccScale=<0.3 0.3 0.3>
    componentLinearVelScale=<0.3 0.3 0.3>
    gravityOverride {z=-980}
    physicsBodyDefinitions [{
      collisionType=CustomSphere
      sphereCollisionRadius=1
      boundBone {boneName=dyn_hair_01}
      boxExtents=<5 20 20>
      constraintSetup {angularLimitsMax=<1 2 10>, angularLimitsMin=<-1 -20 -10>, angularTarget=<0 0 0>}
      localJointOffset=<3 0 0>
     }]
    relativeSpaceBone {boneName=root}
    sphericalLimits [{limitRadius=8.5, drivingBone {boneName=head}, sphereLocalOffset=<0 3 0>}]
   }
   n5 = Anim Dynamics - Bone: dyn_hair_02!(ComponentPose=@n4, Alpha=1, GravityOverride:Vector="0,0,-980", ExternalForce) -> (Pose) {
    angularDampingOverride=0.9999
    angularSpringConstant=500
    bAngularSpring=true
    bGravityOverrideInSimSpace=true
    bLinearSpring=true
    bOverrideAngularDamping=true
    bUseGravityOverride=true
    bUseSphericalLimits=true
    linearSpringConstant=50
    simulationSpace=BoneRelative
    alphaBoolBlend {blendOption=Linear}
    boundBone {boneName=dyn_hair_02}
    componentAppliedLinearAccClamp=<1500 1500 0>
    componentLinearAccScale=<0.3 0.3 0.3>
    componentLinearVelScale=<0.3 0.3 0.3>
    gravityOverride {z=-980}
    physicsBodyDefinitions [{
      collisionType=CustomSphere
      sphereCollisionRadius=1
      boundBone {boneName=dyn_hair_02}
      boxExtents=<5 20 20>
      constraintSetup {angularLimitsMax=<1 2 10>, angularLimitsMin=<-1 -35 -10>, angularTarget=<0 0 0>}
      localJointOffset=<3 0 0>
     }]
    relativeSpaceBone {boneName=root}
    sphericalLimits [{limitRadius=8.5, drivingBone {boneName=head}, sphereLocalOffset=<0 3 0>}]
   }
   n6 = Anim Dynamics - Bone: dyn_hair_03!(ComponentPose=@n5, Alpha=1, GravityOverride:Vector="0,0,-980", ExternalForce="0,0,0") -> (Pose) {
    angularDampingOverride=0.9999
    angularSpringConstant=500
    bAngularSpring=true
    bGravityOverrideInSimSpace=true
    bLinearSpring=true
    bOverrideAngularDamping=true
    bUseGravityOverride=true
    bUseSphericalLimits=true
    linearSpringConstant=50
    simulationSpace=BoneRelative
    alphaBoolBlend {blendOption=Linear}
    boundBone {boneName=dyn_hair_03}
    componentAppliedLinearAccClamp=<1500 1500 0>
    componentLinearAccScale=<0.3 0.3 0.3>
    componentLinearVelScale=<0.3 0.3 0.3>
    gravityOverride {z=-980}
    physicsBodyDefinitions [{
      collisionType=CustomSphere
      sphereCollisionRadius=1
      boundBone {boneName=dyn_hair_03}
      boxExtents=<5 20 20>
      constraintSetup {angularLimitsMax=<1 2 10>, angularLimitsMin=<-1 -35 -10>, angularTarget=<0 0 0>}
      localJointOffset=<3 0 0>
     }]
    relativeSpaceBone {boneName=root}
    sphericalLimits [{limitRadius=14, drivingBone {boneName=spine_05}, sphereLocalOffset=<0 3 0>}]
   }
   n7 = Anim Dynamics - Bone: dyn_hair_04!(ComponentPose=@n6, Alpha=1, GravityOverride:Vector="0,0,-980", ExternalForce="0,0,0") -> (Pose) {
    angularDampingOverride=0.9999
    angularSpringConstant=500
    bAngularSpring=true
    bGravityOverrideInSimSpace=true
    bLinearSpring=true
    bOverrideAngularDamping=true
    bUseGravityOverride=true
    bUseSphericalLimits=true
    linearSpringConstant=50
    simulationSpace=BoneRelative
    alphaBoolBlend {blendOption=Linear}
    boundBone {boneName=dyn_hair_04}
    componentAppliedLinearAccClamp=<1500 1500 0>
    componentLinearAccScale=<0.3 0.3 0.3>
    componentLinearVelScale=<0.3 0.3 0.3>
    gravityOverride {z=-980}
    physicsBodyDefinitions [{
      collisionType=CustomSphere
      sphereCollisionRadius=1
      boundBone {boneName=dyn_hair_04}
      boxExtents=<5 20 20>
      constraintSetup {angularLimitsMax=<1 2 25>, angularLimitsMin=<-1 -45 -25>, angularTarget=<0 0 0>}
      localJointOffset=<3 0 0>
     }]
    relativeSpaceBone {boneName=root}
    sphericalLimits [{limitRadius=8.5, drivingBone {boneName=spine_05}, sphereLocalOffset=<0 0 0>}]
   }
   n8 = Anim Dynamics - Bone: dyn_hair_05!(ComponentPose=@n7, Alpha=1, GravityOverride:Vector="0,0,-980", ExternalForce="0,0,0") -> (Pose) {
    angularDampingOverride=0.9999
    angularSpringConstant=500
    bAngularSpring=true
    bGravityOverrideInSimSpace=true
    bLinearSpring=true
    bOverrideAngularDamping=true
    bUseGravityOverride=true
    bUseSphericalLimits=true
    linearSpringConstant=50
    simulationSpace=BoneRelative
    alphaBoolBlend {blendOption=Linear}
    boundBone {boneName=dyn_hair_05}
    componentAppliedLinearAccClamp=<1500 1500 0>
    componentLinearAccScale=<0.3 0.3 0.3>
    componentLinearVelScale=<0.3 0.3 0.3>
    gravityOverride {z=-980}
    physicsBodyDefinitions [{
      collisionType=CustomSphere
      sphereCollisionRadius=1
      boundBone {boneName=dyn_hair_05}
      boxExtents=<5 20 20>
      constraintSetup {angularLimitsMax=<1 2 25>, angularLimitsMin=<-1 -45 -25>, angularTarget=<0 0 0>}
      localJointOffset=<3 0 0>
     }]
    relativeSpaceBone {boneName=root}
    sphericalLimits [{limitRadius=10, drivingBone {boneName=spine_04}, sphereLocalOffset=<0 0 0>}]
   }
   n9 = Anim Dynamics - Bone: dyn_hair_06!(ComponentPose=@n8, Alpha=1, GravityOverride:Vector="0,0,-980", ExternalForce="0,0,0") -> (Pose) {
    angularDampingOverride=0.9999
    angularSpringConstant=500
    bAngularSpring=true
    bGravityOverrideInSimSpace=true
    bLinearSpring=true
    bOverrideAngularDamping=true
    bUseGravityOverride=true
    bUseSphericalLimits=true
    linearSpringConstant=50
    simulationSpace=BoneRelative
    alphaBoolBlend {blendOption=Linear}
    boundBone {boneName=dyn_hair_06}
    componentAppliedLinearAccClamp=<1500 1500 0>
    componentLinearAccScale=<0.3 0.3 0.3>
    componentLinearVelScale=<0.3 0.3 0.3>
    gravityOverride {z=-980}
    physicsBodyDefinitions [{
      collisionType=CustomSphere
      sphereCollisionRadius=1
      boundBone {boneName=dyn_hair_06}
      boxExtents=<5 20 20>
      constraintSetup {angularLimitsMax=<1 2 25>, angularLimitsMin=<-1 -45 -25>, angularTarget=<0 0 0>}
      localJointOffset=<3 0 0>
     }]
    relativeSpaceBone {boneName=root}
    sphericalLimits [{limitRadius=10.5, drivingBone {boneName=spine_03}, sphereLocalOffset=<0 0 0>}]
   }
   n10 = Anim Dynamics - Bone: dyn_hair_07!(ComponentPose=@n9, Alpha=1, GravityOverride:Vector="0,0,-980", ExternalForce="0,0,0") -> (Pose) {
    angularDampingOverride=0.9999
    angularSpringConstant=500
    bAngularSpring=true
    bGravityOverrideInSimSpace=true
    bLinearSpring=true
    bOverrideAngularDamping=true
    bUseGravityOverride=true
    bUseSphericalLimits=true
    linearSpringConstant=50
    simulationSpace=BoneRelative
    alphaBoolBlend {blendOption=Linear}
    boundBone {boneName=dyn_hair_07}
    componentAppliedLinearAccClamp=<1500 1500 0>
    componentLinearAccScale=<0.3 0.3 0.3>
    componentLinearVelScale=<0.3 0.3 0.3>
    gravityOverride {z=-980}
    physicsBodyDefinitions [{
      collisionType=CustomSphere
      sphereCollisionRadius=1
      boundBone {boneName=dyn_hair_07}
      boxExtents=<5 20 20>
      constraintSetup {angularLimitsMax=<1 2 25>, angularLimitsMin=<-1 -45 -25>, angularTarget=<0 0 0>}
      localJointOffset=<3 0 0>
     }]
    relativeSpaceBone {boneName=root}
    sphericalLimits [{limitRadius=11, drivingBone {boneName=spine_02}, sphereLocalOffset=<0 0 0>}]
   }
   n11 = Anim Dynamics - Bone: dyn_hair_08!(ComponentPose=@n10, Alpha=1, GravityOverride:Vector="0,0,-980", ExternalForce="0,0,0") -> (Pose) {
    angularDampingOverride=0.9999
    angularSpringConstant=500
    bAngularSpring=true
    bGravityOverrideInSimSpace=true
    bLinearSpring=true
    bOverrideAngularDamping=true
    bUseGravityOverride=true
    bUseSphericalLimits=true
    linearSpringConstant=50
    simulationSpace=BoneRelative
    alphaBoolBlend {blendOption=Linear}
    boundBone {boneName=dyn_hair_08}
    componentAppliedLinearAccClamp=<1500 1500 0>
    componentLinearAccScale=<0.3 0.3 0.3>
    componentLinearVelScale=<0.3 0.3 0.3>
    gravityOverride {z=-980}
    physicsBodyDefinitions [{
      collisionType=CustomSphere
      sphereCollisionRadius=1
      boundBone {boneName=dyn_hair_08}
      boxExtents=<5 20 20>
      constraintSetup {angularLimitsMax=<1 2 25>, angularLimitsMin=<-1 -45 -25>, angularTarget=<0 0 0>}
      localJointOffset=<3 0 0>
     }]
    relativeSpaceBone {boneName=root}
    sphericalLimits [{limitRadius=12, drivingBone {boneName=spine_01}, sphereLocalOffset=<0 0 0>}]
   }
   n12 = Anim Dynamics - Bone: dyn_hair_r_01!(ComponentPose=@n11, Alpha=1, GravityOverride:Vector="0,0,-980", ExternalForce="0,0,0") -> (Pose) {
    angularDampingOverride=0.9999
    angularSpringConstant=500
    bAngularSpring=true
    bGravityOverrideInSimSpace=true
    bLinearSpring=true
    bOverrideAngularDamping=true
    bUseGravityOverride=true
    bUseSphericalLimits=true
    linearSpringConstant=50
    simulationSpace=BoneRelative
    alphaBoolBlend {blendOption=Linear}
    boundBone {boneName=dyn_hair_r_01}
    componentAppliedLinearAccClamp=<1500 1500 0>
    componentLinearAccScale=<0.3 0.3 0.3>
    componentLinearVelScale=<0.3 0.3 0.3>
    gravityOverride {z=-980}
    physicsBodyDefinitions [{
      collisionType=CustomSphere
      sphereCollisionRadius=1
      boundBone {boneName=dyn_hair_r_01}
      boxExtents=<5 20 20>
      constraintSetup {angularLimitsMax=<1 2 2>, angularLimitsMin=<-1 -20 -10>, angularTarget=<0 0 0>}
      localJointOffset=<3 0 0>
     }]
    relativeSpaceBone {boneName=root}
    sphericalLimits [{limitRadius=5, drivingBone {boneName=head}, sphereLocalOffset=<0 3 0>}]
   }
   n13 = Anim Dynamics - Bone: dyn_hair_r_02!(ComponentPose=@n12, Alpha=1, GravityOverride:Vector="0,0,-980", ExternalForce="0,0,0") -> (Pose) {
    angularDampingOverride=0.9999
    angularSpringConstant=500
    bAngularSpring=true
    bGravityOverrideInSimSpace=true
    bLinearSpring=true
    bOverrideAngularDamping=true
    bUseGravityOverride=true
    bUseSphericalLimits=true
    linearSpringConstant=50
    simulationSpace=BoneRelative
    alphaBoolBlend {blendOption=Linear}
    boundBone {boneName=dyn_hair_r_02}
    componentAppliedLinearAccClamp=<1500 1500 0>
    componentLinearAccScale=<0.3 0.3 0.3>
    componentLinearVelScale=<0.3 0.3 0.3>
    gravityOverride {z=-980}
    physicsBodyDefinitions [{
      collisionType=CustomSphere
      sphereCollisionRadius=1
      boundBone {boneName=dyn_hair_r_02}
      boxExtents=<5 20 20>
      constraintSetup {angularLimitsMax=<1 2 2>, angularLimitsMin=<-1 -20 -10>, angularTarget=<0 0 0>}
      localJointOffset=<3 0 0>
     }]
    relativeSpaceBone {boneName=root}
    sphericalLimits [{limitRadius=5, drivingBone {boneName=head}, sphereLocalOffset=<0 3 0>}]
   }
   n14 = "Transform (Modify) Bone - Bone: dyn_hair_02"!(ComponentPose=@n13, Alpha=1, Translation="-2,-2,8", Rotation="-20,0,0") -> (Pose) {
    rotationMode=BMM_Additive
    rotationSpace=BCS_BoneSpace
    translationMode=BMM_Additive
    translationSpace=BCS_BoneSpace
    alphaBoolBlend {blendOption=Linear}
    alphaScaleBiasClamp {bInterpResult=true, interpSpeedDecreasing=15, interpSpeedIncreasing=12}
    boneToModify {boneName=dyn_hair_02}
    rotation {pitch=-20}
    translation=<-2 -2 8>
   }
   n15 = "Transform (Modify) Bone - Bone: dyn_hair_03"!(ComponentPose=@n14, Alpha=1) -> (Pose) {
    rotationMode=BMM_Additive
    rotationSpace=BCS_BoneSpace
    scaleMode=BMM_Additive
    scaleSpace=BCS_BoneSpace
    translationMode=BMM_Additive
    translationSpace=BCS_BoneSpace
    alphaBoolBlend {blendOption=Linear}
    alphaScaleBiasClamp {bInterpResult=true, interpSpeedDecreasing=15, interpSpeedIncreasing=12}
    boneToModify {boneName=dyn_hair_03}
    translation {z=3}
   }
   n16 = "Transform (Modify) Bone - Bone: dyn_hair_r_MSTR"!(ComponentPose=@n15, Alpha=1) -> (Pose) {
    rotationMode=BMM_Additive
    rotationSpace=BCS_BoneSpace
    scaleMode=BMM_Additive
    scaleSpace=BCS_BoneSpace
    translationMode=BMM_Additive
    translationSpace=BCS_BoneSpace
    alphaBoolBlend {blendOption=Linear}
    alphaScaleBiasClamp {bInterpResult=true, interpSpeedDecreasing=15, interpSpeedIncreasing=12}
    boneToModify {boneName=dyn_hair_r_MSTR}
    rotation {yaw=-15}
    translation {z=3}
   }
   n17 = "Transform (Modify) Bone - Bone: dyn_hair_l_MSTR"!(ComponentPose=@n16, Alpha=1) -> (Pose) {
    rotationMode=BMM_Additive
    rotationSpace=BCS_BoneSpace
    scaleMode=BMM_Additive
    scaleSpace=BCS_BoneSpace
    translationMode=BMM_Additive
    translationSpace=BCS_BoneSpace
    alphaBoolBlend {blendOption=Linear}
    alphaScaleBiasClamp {bInterpResult=true, interpSpeedDecreasing=15, interpSpeedIncreasing=12}
    boneToModify {boneName=dyn_hair_l_MSTR}
    rotation {yaw=15}
    translation {z=3}
   }
   n18 = "Transform (Modify) Bone - Bone: dyn_hair_04"!(ComponentPose=@n17, Alpha=1, Rotation="30,0,0") -> (Pose) {
    rotationMode=BMM_Additive
    rotationSpace=BCS_BoneSpace
    translationSpace=BCS_BoneSpace
    alphaBoolBlend {blendOption=Linear}
    alphaScaleBiasClamp {bInterpResult=true, interpSpeedDecreasing=15, interpSpeedIncreasing=12}
    boneToModify {boneName=dyn_hair_04}
    rotation {pitch=30}
   }
   n19 = "Constraint - dyn_hair_03 (PPP)"!(ComponentPose=@n18, Alpha=1, ConstraintWeights_0=0.9959, ConstraintWeights_1:float=0.002, ConstraintWeights_2:float=0.002) -> (Pose) {
    alphaBoolBlend {blendOption=Linear}
    alphaScaleBiasClamp {bClampResult=true, bInterpResult=true, clampMax=0.5, interpSpeedDecreasing=5, interpSpeedIncreasing=7}
    boneToModify {boneName=dyn_hair_03}
    constraintSetup [{targetBone {boneName=spine_05}}, {targetBone {boneName=clavicle_l}}, {targetBone {boneName=clavicle_r}}]
    constraintWeights [1,1,1]
   }
   n20 = "Constraint - dyn_hair_04 (PPPP)"!(
    ComponentPose=@n19
    Alpha=1
    ConstraintWeights_0=0.5
    ConstraintWeights_1:float=0.1
    ConstraintWeights_2:float=0.2
    ConstraintWeights_3:float=0.2
   ) -> (Pose) {
    alphaBoolBlend {blendOption=Linear}
    alphaScaleBiasClamp {bClampResult=true, bInterpResult=true, clampMax=0.3, interpSpeedDecreasing=5, interpSpeedIncreasing=7}
    boneToModify {boneName=dyn_hair_04}
    constraintSetup [
     {targetBone {boneName=spine_05}}
     {targetBone {boneName=spine_04}}
     {targetBone {boneName=dyn_horn_l_01}}
     {targetBone {boneName=dyn_horn_r_01}}
    ]
    constraintWeights [1,1,1,1]
   }
   n21 = "Constraint - dyn_hair_04 (PP)"!(ComponentPose=@n20, Alpha=1, ConstraintWeights_0=0.5, ConstraintWeights_1:float=0.5) -> (Pose) {
    alphaBoolBlend {blendOption=Linear}
    alphaScaleBiasClamp {bInterpResult=true, interpSpeedDecreasing=5, interpSpeedIncreasing=7}
    boneToModify {boneName=dyn_hair_04}
    constraintSetup [{targetBone {boneName=spine_04}}, {targetBone {boneName=spine_05}}]
    constraintWeights [1,1]
   }
   n22 = "Constraint - dyn_hair_05 (PP)"!(ComponentPose=@n21, Alpha=1, ConstraintWeights_0=0.5, ConstraintWeights_1:float=0.5) -> (Pose) {
    alphaBoolBlend {blendOption=Linear}
    alphaScaleBiasClamp {bInterpResult=true, interpSpeedDecreasing=5, interpSpeedIncreasing=7}
    boneToModify {boneName=dyn_hair_05}
    constraintSetup [{targetBone {boneName=spine_05}}, {targetBone {boneName=spine_04}}]
    constraintWeights [1,1]
   }
   n23 = "Constraint - dyn_hair_02 (P)"(ComponentPose=@n22, Alpha=1, ConstraintWeights_0=1) {
    alphaBoolBlend {blendOption=Linear}
    alphaScaleBiasClamp {bClampResult=true, clampMax=0.7}
    boneToModify {boneName=dyn_hair_02}
    constraintSetup [{targetBone {boneName=neck_02}}]
    constraintWeights [1]
   }
   n24 = "Constraint - dyn_hair_03 (P)"(ComponentPose=@n23, Alpha=1, ConstraintWeights_0=1) {
    alphaBoolBlend {blendOption=Linear}
    alphaScaleBiasClamp {bClampResult=true, clampMax=0.7}
    boneToModify {boneName=dyn_hair_03}
    constraintSetup [{targetBone {boneName=neck_01}}]
    constraintWeights [1]
   }
   n25 = "Constraint - dyn_hair_04 (P)"(ComponentPose=@n24, Alpha=1, ConstraintWeights_0=1) {
    alphaBoolBlend {blendOption=Linear}
    boneToModify {boneName=dyn_hair_04}
    constraintSetup [{targetBone {boneName=spine_05}}]
    constraintWeights [1]
   }
   n26 = "Constraint - dyn_hair_05 (P)"(ComponentPose=@n25, Alpha=1, ConstraintWeights_0=1) {
    alphaBoolBlend {blendOption=Linear}
    boneToModify {boneName=dyn_hair_05}
    constraintSetup [{targetBone {boneName=spine_05}}]
    constraintWeights [1]
   }
   n27 = "Constraint - dyn_hair_06 (P)"(ComponentPose=@n26, Alpha=0.5, ConstraintWeights_0=1) {
    alphaBoolBlend {blendOption=Linear}
    boneToModify {boneName=dyn_hair_06}
    constraintSetup [{targetBone {boneName=spine_04}}]
    constraintWeights [1]
   }
   n28 = "Constraint - dyn_hair_07 (P)"(ComponentPose=@n27, Alpha=0.5, ConstraintWeights_0=1) {
    alphaBoolBlend {blendOption=Linear}
    boneToModify {boneName=dyn_hair_07}
    constraintSetup [{targetBone {boneName=spine_03}}]
    constraintWeights [1]
   }
   n29 = "Constraint - dyn_hair_l_01 (P)"(ComponentPose=@n28, Alpha=1, ConstraintWeights_0=1) {
    alphaBoolBlend {blendOption=Linear}
    alphaScaleBiasClamp {bClampResult=true, clampMax=0.2}
    boneToModify {boneName=dyn_hair_l_01}
    constraintSetup [{targetBone {boneName=neck_02}}]
    constraintWeights [1]
   }
   n30 = "Constraint - dyn_hair_r_01 (P)"(ComponentPose=@n29, Alpha=1, ConstraintWeights_0=1) {
    alphaBoolBlend {blendOption=Linear}
    alphaScaleBiasClamp {bClampResult=true, clampMax=0.2}
    boneToModify {boneName=dyn_hair_r_01}
    constraintSetup [{targetBone {boneName=neck_02}}]
    constraintWeights [1]
   }
   n31 = "Constraint - dyn_hair_l_02 (P)"(ComponentPose=@n30, Alpha=1, ConstraintWeights_0=1) {
    alphaBoolBlend {blendOption=Linear}
    alphaScaleBiasClamp {bClampResult=true, clampMax=0.6}
    boneToModify {boneName=dyn_hair_l_02}
    constraintSetup [{targetBone {boneName=spine_05}}]
    constraintWeights [1]
   }
   n32 = "Constraint - dyn_hair_r_02 (P)"(ComponentPose=@n31, Alpha=1, ConstraintWeights_0=1) {
    alphaBoolBlend {blendOption=Linear}
    alphaScaleBiasClamp {bClampResult=true, clampMax=0.6}
    boneToModify {boneName=dyn_hair_r_02}
    constraintSetup [{targetBone {boneName=spine_05}}]
    constraintWeights [1]
   }
   n33 = "Constraint - dyn_hair_04 (PPP)"!(ComponentPose=@n32, Alpha=1, ConstraintWeights_0=0.2, ConstraintWeights_1:float=0.4, ConstraintWeights_2:float=0.4) -> (Pose) {
    alphaBoolBlend {blendOption=Linear}
    alphaScaleBiasClamp {bClampResult=true, clampMax=0.2}
    boneToModify {boneName=dyn_hair_04}
    constraintSetup [{targetBone {boneName=spine_05}}, {targetBone {boneName=dyn_horn_r_02}}, {targetBone {boneName=dyn_horn_l_02}}]
    constraintWeights [1,1,1]
   }
   n34 = "Constraint - dyn_hair_04 (P)"(ComponentPose=@n33, Alpha=1, ConstraintWeights_0=1) {
    alphaBoolBlend {blendOption=Linear}
    boneToModify {boneName=dyn_hair_04}
    constraintSetup [{targetBone {boneName=spine_05}}]
    constraintWeights [1]
   }
   n35 = "Constraint - dyn_hair_05 (P)"(ComponentPose=@n34, Alpha=1, ConstraintWeights_0=1) {
    alphaBoolBlend {blendOption=Linear}
    boneToModify {boneName=dyn_hair_05}
    constraintSetup [{targetBone {boneName=spine_05}}]
    constraintWeights [1]
   }
   n36 = "Constraint - dyn_hair_06 (P)"(ComponentPose=@n35, Alpha=0.5, ConstraintWeights_0=1) {
    alphaBoolBlend {blendOption=Linear}
    boneToModify {boneName=dyn_hair_06}
    constraintSetup [{targetBone {boneName=spine_04}}]
    constraintWeights [1]
   }
   n37 = "Constraint - dyn_hair_07 (P)"(ComponentPose=@n36, Alpha=0.5, ConstraintWeights_0=1) {
    alphaBoolBlend {blendOption=Linear}
    boneToModify {boneName=dyn_hair_07}
    constraintSetup [{targetBone {boneName=spine_03}}]
    constraintWeights [1]
   }
   n38 = "Constraint - dyn_hair_08 (P)"(ComponentPose=@n37, Alpha=0.5, ConstraintWeights_0=1) {
    alphaBoolBlend {blendOption=Linear}
    boneToModify {boneName=dyn_hair_08}
    constraintSetup [{targetBone {boneName=spine_03}}]
    constraintWeights [1]
   }
   n39 = "Transform (Modify) Bone - Bone: dyn_hair_03"!(ComponentPose=@n38, Alpha=1, Rotation:Rotator="-8,0,0") -> (Pose) {
    rotationMode=BMM_Replace
    rotationSpace=BCS_BoneSpace
    alphaBoolBlend {blendOption=Linear}
    boneToModify {boneName=dyn_hair_03}
    rotation {pitch=-8}
   }
   n40 = "Constraint - dyn_hair_02 (PP)"!(ComponentPose=@n39, Alpha=1, ConstraintWeights_0=0.7, ConstraintWeights_1:float=0.3) -> (Pose) {
    alphaBoolBlend {blendOption=Linear}
    alphaScaleBiasClamp {bClampResult=true, clampMax=0.2, interpSpeedDecreasing=5, interpSpeedIncreasing=12}
    boneToModify {boneName=dyn_hair_02}
    constraintSetup [{targetBone {boneName=spine_05}}, {targetBone {boneName=head}}]
    constraintWeights [1,1]
   }
   n41 = "Constraint - dyn_hair_03 (PP)"!(ComponentPose=@n40, Alpha=1, ConstraintWeights_0=0.8, ConstraintWeights_1:float=0.2) -> (Pose) {
    alphaBoolBlend {blendOption=Linear}
    alphaScaleBiasClamp {bClampResult=true, clampMax=0.4, interpSpeedDecreasing=5, interpSpeedIncreasing=12}
    boneToModify {boneName=dyn_hair_03}
    constraintSetup [{targetBone {boneName=spine_05}}, {targetBone {boneName=spine_04}}]
    constraintWeights [1,1]
   }
   n42 = "Constraint - dyn_hair_05 (PP)"(ComponentPose=@n41, Alpha=1, ConstraintWeights_0=0.7, ConstraintWeights_1=0.3) {
    alphaBoolBlend {blendOption=Linear}
    alphaScaleBiasClamp {bClampResult=true, bInterpResult=true, clampMax=0.3, interpSpeedDecreasing=5, interpSpeedIncreasing=7}
    boneToModify {boneName=dyn_hair_05}
    constraintSetup [{targetBone {boneName=spine_04}}, {targetBone {boneName=spine_03}}]
    constraintWeights [1,1]
   }
   n43 = "Constraint - dyn_hair_l_02 (PP)"!(ComponentPose=@n42, Alpha=0.5, ConstraintWeights_0=0.7, ConstraintWeights_1:float=0.3) -> (Pose) {
    alphaBoolBlend {blendOption=Linear}
    alphaScaleBiasClamp {bClampResult=true, bInterpResult=true, clampMax=0.2, interpSpeedDecreasing=5, interpSpeedIncreasing=7}
    boneToModify {boneName=dyn_hair_l_02}
    constraintSetup [{targetBone {boneName=clavicle_l}}, {targetBone {boneName=dyn_hair_03}}]
    constraintWeights [1,1]
   }
   n44 = "Constraint - dyn_hair_r_02 (PP)"!(ComponentPose=@n43, Alpha=0.5, ConstraintWeights_0=0.7, ConstraintWeights_1:float=0.3) -> (Pose) {
    alphaBoolBlend {blendOption=Linear}
    alphaScaleBiasClamp {bClampResult=true, bInterpResult=true, clampMax=0.2, interpSpeedDecreasing=5, interpSpeedIncreasing=7}
    boneToModify {boneName=dyn_hair_r_02}
    constraintSetup [{targetBone {boneName=clavicle_r}}, {targetBone {boneName=dyn_hair_03}}]
    constraintWeights [1,1]
   }
   n45 = "Transform (Modify) Bone - Bone: dyn_hair_04"(ComponentPose=@n44, Alpha=1, Rotation="10,0,0") {
    rotationMode=BMM_Replace
    rotationSpace=BCS_BoneSpace
    scaleSpace=BCS_BoneSpace
    translationMode=BMM_Additive
    translationSpace=BCS_BoneSpace
    alphaBoolBlend {blendOption=Linear}
    boneToModify {boneName=dyn_hair_04}
    rotation {pitch=10}
    translation {z=2}
   }
   n46 = "Transform (Modify) Bone - Bone: dyn_hair_06"!(ComponentPose=@n45, Alpha=1, Rotation="15,0,0") -> (Pose) {
    rotationMode=BMM_Replace
    rotationSpace=BCS_BoneSpace
    scaleSpace=BCS_BoneSpace
    translationMode=BMM_Additive
    translationSpace=BCS_BoneSpace
    alphaBoolBlend {blendOption=Linear}
    boneToModify {boneName=dyn_hair_06}
    rotation {pitch=15}
   }
   Output Pose(Result=Component To Local(ComponentPose=@n46)) {layerGroup=None}
   n47 = Layered blend per bone(
    BasePose=Copy Pose From Mesh() {bUseAttachedParent=true}
    BlendPoses_0=Copy Pose From Mesh(SourceMeshComponent=Get body_mesh() -> (body_mesh:Skeletal Mesh Component Object Reference))
   ) {layerSetup [{branchFilters [{boneName=clavicle_l}, {boneName=dyn_horn_l_01}, {boneName=clavicle_r}, {boneName=dyn_horn_r_01}]}]}
   n48 = Control Rig!(
    Source=@n47
    Alpha
    jump:double=Get Jump() -> (Jump:double)
    Crouch=Get Crouch()
    ADS:double=Get Target() -> (Target:double)
    Skydive=Get Skydive()
    Parachute:double=Get Parachute() -> (Parachute:double)
    Backbling:double=Get BackBling() -> (BackBling:double)
    moving:double=Get moving()
    Emote:bool=Get bIsPlayingEmote()
   ) -> (Pose) {
    alphaBoolBlend {blendOption=Linear}
    controlRigAssetReference {blueprintRigClass=&/F_MED_PastaSauceMarinara_FaceAcc_CtrlRig.F_MED_PastaSauceMarinara_FaceAcc_CtrlRig_C}
   }
   Save cached pose 'CTrl'(Pose=@n48) {cacheName=CTrl}
   n49 = RigidBody!(
    ComponentPose=Local To Component(LocalPose=Use cached pose 'CTrl'())
    Alpha
    OverrideWorldGravity:Vector=Get GravityOverride()
    ExternalForce
   ) -> (Pose) {
    bClampLinearTranslationLimitToRefPose=true
    bOverrideWorldGravity=true
    overridePhysicsAsset=&/F_MED_PastaSauceMarinara_FaceAcc_Physics
    simulationSpace=BaseBoneSpace
    alphaBoolBlend {blendOption=Linear}
    baseBoneRef {boneName=root}
    componentAppliedLinearAccClamp=<5000 5000 5000>
    componentLinearAccScale=<0.2 0.4 0.4>
    componentLinearVelScale=<0.2 0.4 0.4>
    overrideWorldGravity {z=-980}
   }
   n50 = Component To Local(ComponentPose=@n49)
   n51 = RigidBody(ComponentPose=Local To Component(LocalPose=Use cached pose 'CTrl'()), OverrideWorldGravity=Get GravityOverride()) {
    bClampLinearTranslationLimitToRefPose=true
    bOverrideWorldGravity=true
    overridePhysicsAsset=&/F_MED_PastaSauceMarinara_FaceAcc_Physics_Skydive
    simulationSpace=BaseBoneSpace
    alphaBoolBlend {blendOption=Linear}
    baseBoneRef {boneName=root}
    componentAppliedLinearAccClamp=<5000 5000 5000>
    componentLinearAccScale=<0.2 0.4 0.4>
    componentLinearVelScale=<0.2 0.4 0.4>
    overrideWorldGravity {z=-980}
   }
   n52 = Component To Local(ComponentPose=@n51)
   n53 = Get bIsSkyDiving()
   n54 = RigidBody(ComponentPose=Local To Component(LocalPose=Use cached pose 'CTrl'()), OverrideWorldGravity=Get GravityOverride()) {
    bClampLinearTranslationLimitToRefPose=true
    bOverrideWorldGravity=true
    overridePhysicsAsset=&F_MED_PastaSauceMarinara_FaceAcc_Physics
    simulationSpace=BaseBoneSpace
    alphaBoolBlend {blendOption=Linear}
    baseBoneRef {boneName=spine_01}
    componentLinearAccScale=<0.5 0.5 0.5>
    componentLinearVelScale=<0.5 0.5 0.5>
    overrideWorldGravity {z=-980}
   }
   Save cached pose 'RBAN'(Pose=Blend Poses by bool(
     BlendPose_0=Component To Local(ComponentPose=@n54)
     BlendPose_1=Blend Poses by bool(BlendPose_0=@n52, BlendPose_1=@n50, BlendTime_0=0.5, BlendTime_1=0.4, bActiveValue=@n53) {blendType=Linear}
     BlendTime_0=0.5
     BlendTime_1=0.4
     bActiveValue=Get bIsCrouching()
    ) {blendType=Linear}) {cacheName=RBAN}
  ]}
 EventGraph {nodes [
   Event BlueprintUpdateAnimation()
   > n1 = Sequence!(execute) -> (then_0, then_1, then_2:Exec, then_3:Exec, then_4:Exec)
   > n2 = Gravity Override State Function DEPRECATED(CharacterPartAnimInstance=Self-Reference(), GameTimeInSeconds=GetGameTimeInSeconds(), GravityOverrideParams=Get Gravity Override Params())
   > Set GravityOverride(GravityOverride=@n2)
   n3 = Cast To PlayerPawn_Athena(execute=@n1.then_1, Object=TryGetPawnOwner())
   > Set body_mesh(body_mesh:Skeletal Mesh Component Object Reference=GetSkeletalMeshForPartType(self=@n3)) -> (Output_Get:Skeletal Mesh Component Object Reference)
   n4 = CheckForBackbling()
   Set WearingBackbling(execute=@n1.then_2, WearingBackbling:bool=OR Boolean!(A=@n4, B=false, C:bool=@n4.bIsBackblingLong) -> (ReturnValue)) -> (Output_Get:bool)
   > Set Cape(Cape:bool=@n4.bIsBackblingACape) -> (Output_Get:bool)
   > Set BackBling(BackBling:double=SelectFloat(A=1, B=0, bPickA=OR Boolean(A=@n4, B=@n4.bIsBackblingACape, C=@n4.bIsBackblingLong))) -> (Output_Get:double)
   n5 = Get bIsJumping()
   n6 = Get bIsFalling()
   Set Jump(execute=@n1.then_3, Jump:double=SelectFloat(A=1, B=0, bPickA=OR Boolean!(A=@n5, B=@n6) -> (ReturnValue))) -> (Output_Get:double)
   > Set Target(Target:double=SelectFloat(A=1, B=0, bPickA=Get bIsTargeting())) -> (Output_Get:double)
   > Set Crouch(Crouch:double=SelectFloat(A=1, B=0, bPickA=Get bIsCrouching())) -> (Output_Get:double)
   > Set moving(moving:double=SelectFloat(A=1, B=0, bPickA=Get bIsMoving2D())) -> (Output_Get:double)
   > Set Skydive(Skydive:double=SelectFloat(A=1, B=0, bPickA=Get bIsSkydiveFloating())) -> (Output_Get:double)
   n7 = Get bIsParachuteOpened()
   > Set Parachute(Parachute:double=SelectFloat(A=1, B=0, bPickA=@n7)) -> (Output_Get:double)
   Set RBAN Off(execute=@n1.then_4, RBAN Off:double=SelectFloat(A=0, B=1, bPickA=Get bRigidBodyEnabled())) -> (Output_Get:double)
   > Set RBAN On(RBAN On:double=SelectFloat(A=1, B=0, bPickA=Get bRigidBodyEnabled())) -> (Output_Get:double)
   AND Boolean!(A=Get bIsSkyDiving(), B=NOT Boolean(A=Get bIsParachuteOpened()), C:bool=NOT Boolean(A=Get bIsSkydiveDiveMode())) -> (ReturnValue)
   OR Boolean(A=@n7, B=Get bIsSkydiveDiveMode())
  ]}
]
variables [
 AccScaleOverride {type=Vector}
 BackBling {type=number}
 body_mesh {type=SkeletalMeshComponent}
 Cape {type=boolean}
 Crouch {type=number}
 Gravity Override Params {type=GravityOverrideParamsStruct, default {onGround_Falling {gravityOverride=<0 0 0>}}}
 GravityOverride {type=Vector}
 Jump {type=number}
 moving {type=number}
 Parachute {type=number}
 RBAN Off {type=number}
 RBAN On {type=number}
 Skydive {type=number}
 Target {type=number}
 WearingBackbling {type=boolean}
]
